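# Physics asset for the right-wrist chain bracelet on the F_MED_GraveFlannel female character cosmetic.
# Three capsule-based dynamic chain links hang in series from a base bone, with limited swing and zero-degree twist limits; the last link uses two capsules.
# Forearm, forearm twist, hand and thumb capsules approximate the arm around the chain.
# Constraints joining those arm bodies, and the chain base to the forearm twist, leave swing and twist free.
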PhysicsAsset /BRCosmetics/Characters/Player/Female/Medium/Bodies/F_MED_GraveFlannel/Meshes/F_MED_GraveFlannel_Bracelet_Physics
bodies {
 base_wristChain_01_r [{shapeType=Capsule, length=0.505, radius=0.505, shapeName=base_wristChain_01_r_capsule, center=<0.4696 0 0>, rotation=<rot 0 -90 0>}]
 dyn_wristChain_01_r [{
   shapeType=Capsule
   length=0.8529
   radius=0.4349
   shapeName=dyn_wristChain_01_r_capsule
   center=<0.4165 -0.0491 -0.0407>
   rotation=<rot 0 -174.3439 -88.1287>
  }]
 dyn_wristChain_02_r [{
   shapeType=Capsule
   length=1.801
   radius=0.7224
   shapeName=dyn_wristChain_02_r_capsule
   center=<1.3397 0.0718 -0.0237>
   rotation=<rot 0 92.0126 -89.4197>
  }]
 dyn_wristChain_03_r [
  {
   shapeType=Capsule
   length=0.8461
   radius=0.4
   shapeName=dyn_wristChain_03_r_capsule
   center=<1.0292 -1.2653 0.0476>
   rotation=<rot 2.8203 33.8735 -89.2442>
  }
  {
   shapeType=Capsule
   length=2.0402
   radius=0.7695
   shapeName=dyn_wristChain_03_r_capsule
   center=<1.7788 -0.0099 0.1045>
   rotation=<rot 0.2545 103.8609 -87.0909>
  }
 ]
 hand_r [{
   shapeType=Capsule
   length=4.1339
   radius=2.801
   shapeName=hand_r_capsule
   center=<-3.8166 0.7596 0.4665>
   rotation=<rot 15.7974 -89.9999 1.5933>
  }]
 lowerarm_r [{
   shapeType=Capsule
   length=13.8712
   radius=3.1081
   shapeName=lowerarm_r_capsule
   center=<-8.3267 0.3799 0>
   rotation=<rot -0.0001 -89.9999 -90>
  }]
 lowerarm_twist_01_r [{
   shapeType=Capsule
   length=7.15
   radius=2.6819
   shapeName=lowerarm_twist_01_r_capsule
   center=<1.1251 0.1751 -0.0564>
   rotation=<rot 0.0005 91.1503 -94.8358>
  }]
 thumb_01_r [{
   shapeType=Capsule
   length=3.9321
   radius=2.0046
   shapeName=thumb_01_r_capsule
   center=<-2.9792 0.7422 -0.0927>
   rotation=<rot -1.1037 91.0963 -82.2792>
  }]
}
constraints [
 {bone1Name=dyn_wristChain_01_r, bone2Name=base_wristChain_01_r, swing1LimitDegrees=45, swing2LimitDegrees=45, twistLimitDegrees=0}
 {bone1Name=dyn_wristChain_02_r, bone2Name=dyn_wristChain_01_r, swing1LimitDegrees=60, swing2LimitDegrees=60, twistLimitDegrees=0}
 {bone1Name=dyn_wristChain_03_r, bone2Name=dyn_wristChain_02_r, swing1LimitDegrees=60, swing2LimitDegrees=60, twistLimitDegrees=0}
 {
  bone1Name=base_wristChain_01_r
  bone2Name=lowerarm_twist_01_r
  swing1LimitDegrees=45
  swing1Motion=Free
  swing2LimitDegrees=45
  swing2Motion=Free
  twistLimitDegrees=45
  twistMotion=Free
 }
 {
  bone1Name=thumb_01_r
  bone2Name=hand_r
  swing1LimitDegrees=45
  swing1Motion=Free
  swing2LimitDegrees=45
  swing2Motion=Free
  twistLimitDegrees=45
  twistMotion=Free
 }
 {
  bone1Name=hand_r
  bone2Name=lowerarm_r
  swing1LimitDegrees=45
  swing1Motion=Free
  swing2LimitDegrees=45
  swing2Motion=Free
  twistLimitDegrees=45
  twistMotion=Free
 }
 {
  bone1Name=lowerarm_twist_01_r
  bone2Name=lowerarm_r
  swing1LimitDegrees=45
  swing1Motion=Free
  swing2LimitDegrees=45
  swing2Motion=Free
  twistLimitDegrees=45
  twistMotion=Free
 }
]
