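# Physics asset for the GraveFlannel female medium body cosmetic, providing capsule colliders that, by its name, serve the cape.
# Capsules cover pelvis, spine, clavicles, head and feet, while arm, leg and neck bodies have no shapes.
# Six front, side and side-back skirt chains hang from spine_01 with overlapping capsules, the SB pair continuing into strap bones.
# All constraints leave swing and twist free, so they impose no angular limits.
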PhysicsAsset /BRCosmetics/Characters/Player/Female/Medium/Bodies/F_MED_GraveFlannel/Meshes/F_MED_GraveFlannel_CapeColliders
bodies {
 calf_l []
 calf_r []
 clavicle_l [{shapeType=Capsule, length=6.5623, radius=7.5194, center=<3.9566 -1.0862 -2.1127>, rotation=<rot -77.5566 177.3915 7.7314>}]
 clavicle_r [{shapeType=Capsule, length=6.5623, radius=7.5194, center=<-3.9566 1.0862 2.1127>, rotation=<rot -77.5566 177.3915 -172.2686>}]
 dyn_SkirtStrap_06_SB_l [{
   shapeType=Capsule
   length=9.6648
   radius=7.3578
   shapeName=dyn_SkirtStrap_06_SB_l_capsule
   center=<3.1881 -0.0449 4.6139>
   rotation=<rot -0.0001 -0.0032 -90>
  }]
 dyn_SkirtStrap_06_SB_r [{
   shapeType=Capsule
   length=9.6648
   radius=7.3578
   shapeName=dyn_SkirtStrap_06_SB_l_capsule
   center=<3.1881 -0.023 4.6139>
   rotation=<rot -0.0001 -0.0032 -90>
  }]
 dyn_SkirtStrap_07_SB_l [{
   shapeType=Capsule
   length=9.6648
   radius=8.4586
   shapeName=dyn_SkirtStrap_06_SB_l_capsule
   center=<2.6928 -0.0449 4.6139>
   rotation=<rot -0.0001 -0.0032 -90>
  }]
 dyn_SkirtStrap_07_SB_r [{
   shapeType=Capsule
   length=9.6648
   radius=8.4907
   shapeName=dyn_SkirtStrap_06_SB_l_capsule
   center=<2.6928 -0.023 4.6139>
   rotation=<rot -0.0001 -0.0032 -90>
  }]
 dyn_Skirt_01_F_l [{
   shapeType=Capsule
   length=2.5299
   radius=8.801
   shapeName=dyn_Skirt_01_F_l_capsule
   center=<5.8086 0.1943 -2.6771>
   rotation=<rot 0 -92.2646 -89.934>
  }]
 dyn_Skirt_01_F_r [{
   shapeType=Capsule
   length=2.5299
   radius=8.801
   shapeName=dyn_Skirt_01_F_l_capsule
   center=<5.9771 0.1863 -3.2737>
   rotation=<rot 0.015 -92.2976 -91.8892>
  }]
 dyn_Skirt_01_SB_l [{
   shapeType=Capsule
   length=2.8283
   radius=8.5016
   shapeName=dyn_Skirt_01_F_l_capsule
   center=<7.5768 0.1244 -0.9581>
   rotation=<rot 0 -92.2646 -95.7367>
  }]
 dyn_Skirt_01_SB_r [{
   shapeType=Capsule
   length=3.3915
   radius=8.1395
   shapeName=dyn_Skirt_01_F_l_capsule
   center=<7.7877 0.1161 -1.2865>
   rotation=<rot 0 -92.2646 -94.7005>
  }]
 dyn_Skirt_01_S_l [{
   shapeType=Capsule
   length=2.5299
   radius=8.801
   shapeName=dyn_Skirt_01_F_l_capsule
   center=<6.8781 0.1521 -1.6732>
   rotation=<rot -0.0004 -92.2638 -95.6957>
  }]
 dyn_Skirt_01_S_r [{
   shapeType=Capsule
   length=2.5299
   radius=8.801
   shapeName=dyn_Skirt_01_F_l_capsule
   center=<6.2398 0.1773 -2.1904>
   rotation=<rot 0.0001 -92.2648 -95.3404>
  }]
 dyn_Skirt_02_F_l [{
   shapeType=Capsule
   length=4.1138
   radius=9.5712
   shapeName=dyn_Skirt_02_F_l_capsule
   center=<5.4385 -0.2987 -1.9596>
   rotation=<rot 0.0893 -90.0676 -93.8024>
  }]
 dyn_Skirt_02_F_r [{
   shapeType=Capsule
   length=4.1138
   radius=9.5712
   shapeName=dyn_Skirt_02_F_l_capsule
   center=<5.9229 -0.299 -1.7911>
   rotation=<rot 0.0894 -90.068 -93.3024>
  }]
 dyn_Skirt_02_SB_l [{
   shapeType=Capsule
   length=5.4568
   radius=8.6999
   shapeName=dyn_Skirt_02_F_l_capsule
   center=<4.3425 -0.2985 -2.6635>
   rotation=<rot 0.0893 -90.0677 -87.9445>
  }]
 dyn_Skirt_02_SB_r [{
   shapeType=Capsule
   length=5.6039
   radius=8.5529
   shapeName=dyn_Skirt_02_F_l_capsule
   center=<4.0078 -0.2983 -2.8301>
   rotation=<rot 0.0214 -89.4086 -84.1182>
  }]
 dyn_Skirt_02_S_l [{
   shapeType=Capsule
   length=3.5121
   radius=9.4459
   shapeName=dyn_Skirt_02_F_l_capsule
   center=<4.2963 -0.298 -2.3895>
   rotation=<rot 0.0893 -90.0675 -87.1579>
  }]
 dyn_Skirt_02_S_r [{
   shapeType=Capsule
   length=4.5855
   radius=9.5712
   shapeName=dyn_Skirt_02_F_l_capsule
   center=<4.2142 -0.2987 -2.9268>
   rotation=<rot 0.0893 -90.0678 -86.0939>
  }]
 dyn_Skirt_03_F_l [{
   shapeType=Capsule
   length=5.2473
   radius=9.5005
   shapeName=dyn_Skirt_02_F_l_capsule
   center=<6.4125 -0.3009 -2.679>
   rotation=<rot 0.0893 -90.0676 -85.7623>
  }]
 dyn_Skirt_03_F_r [{
   shapeType=Capsule
   length=5.5145
   radius=9.4841
   shapeName=dyn_Skirt_02_F_l_capsule
   center=<6.3431 -0.3004 -2.3864>
   rotation=<rot 0.0894 -90.0681 -83.7773>
  }]
 dyn_Skirt_03_SB_l [{
   shapeType=Capsule
   length=5.2473
   radius=9.5005
   shapeName=dyn_Skirt_02_F_l_capsule
   center=<7.075 -0.3045 -3.8107>
   rotation=<rot 0.1008 -90.0675 -89.777>
  }]
 dyn_Skirt_03_SB_r [{
   shapeType=Capsule
   length=5.2473
   radius=9.5005
   shapeName=dyn_Skirt_02_F_l_capsule
   center=<6.9189 -0.303 -3.6354>
   rotation=<rot 0.0893 -90.0677 -87.9445>
  }]
 dyn_Skirt_03_S_l [{
   shapeType=Capsule
   length=5.1193
   radius=9.6286
   shapeName=dyn_Skirt_02_F_l_capsule
   center=<7.0682 -0.302 -2.8872>
   rotation=<rot 0.0893 -90.0676 -89.5041>
  }]
 dyn_Skirt_03_S_r [{
   shapeType=Capsule
   length=5.2473
   radius=9.5005
   shapeName=dyn_Skirt_02_F_l_capsule
   center=<7.0963 -0.3033 -3.6875>
   rotation=<rot 0.0893 -90.0678 -89.9669>
  }]
 dyn_Skirt_04_F_l [{
   shapeType=Capsule
   length=4.3672
   radius=9.1646
   shapeName=dyn_Skirt_02_F_l_capsule
   center=<6.7038 -1.0988 -2.8369>
   rotation=<rot 0.0002 0.0003 -111.0014>
  }]
 dyn_Skirt_04_F_r [{
   shapeType=Capsule
   length=4.3672
   radius=9.1646
   shapeName=dyn_Skirt_02_F_l_capsule
   center=<6.7038 0.9468 -3.3987>
   rotation=<rot 0 0 -96.1345>
  }]
 dyn_Skirt_04_SB_l [{
   shapeType=Capsule
   length=1.1561
   radius=12.3756
   shapeName=dyn_Skirt_02_F_l_capsule
   center=<4.1559 -0.5707 -1.9179>
   rotation=<rot 0 0.0001 -91.2067>
  }]
 dyn_Skirt_04_SB_r [{
   shapeType=Capsule
   length=1.1561
   radius=12.3756
   shapeName=dyn_Skirt_02_F_l_capsule
   center=<4.1559 0.3153 -1.7891>
   rotation=<rot 0.1354 1.0764 -97.1777>
  }]
 dyn_Skirt_04_S_l [{
   shapeType=Capsule
   length=7.1203
   radius=9.1646
   shapeName=dyn_Skirt_02_F_l_capsule
   center=<6.7038 0.4472 -2.4044>
   rotation=<rot 0 0.0001 -104.7138>
  }]
 dyn_Skirt_04_S_r [{
   shapeType=Capsule
   length=6.3426
   radius=8.719
   shapeName=dyn_Skirt_02_F_l_capsule
   center=<6.7038 0.9429 -3.0795>
   rotation=<rot 0 0.0001 -97.5108>
  }]
 dyn_Skirt_05_F_l [{
   shapeType=Capsule
   length=2.5299
   radius=9.2132
   shapeName=dyn_Skirt_02_F_l_capsule
   center=<8.4817 -1.111 -2.9546>
   rotation=<rot -1.6849 -15.991 -111.6125>
  }]
 dyn_Skirt_05_F_r [{
   shapeType=Capsule
   length=2.5299
   radius=9.4589
   shapeName=dyn_Skirt_02_F_l_capsule
   center=<8.8195 2.7246 -3.0002>
   rotation=<rot 2.7113 12.9066 -101.6634>
  }]
 dyn_Skirt_05_SB_l [{
   shapeType=Capsule
   length=3.3962
   radius=12.9879
   shapeName=dyn_Skirt_02_F_l_capsule
   center=<1.9223 0.2612 -0.6112>
   rotation=<rot 0.3883 1.832 -97.208>
  }]
 dyn_Skirt_05_SB_r [{
   shapeType=Capsule
   length=4.0142
   radius=12.3699
   shapeName=dyn_Skirt_02_F_l_capsule
   center=<1.9239 0.3899 -1.2847>
   rotation=<rot -0.2238 -1.0554 -101.968>
  }]
 dyn_Skirt_05_S_l [{
   shapeType=Capsule
   length=6.3544
   radius=8.6731
   shapeName=dyn_Skirt_02_F_l_capsule
   center=<5.844 0.8624 -2.4093>
   rotation=<rot -3.4492 -14.5832 -103.0205>
  }]
 dyn_Skirt_05_S_r [{
   shapeType=Capsule
   length=5.3839
   radius=8.5199
   shapeName=dyn_Skirt_02_F_l_capsule
   center=<5.7606 0.6643 -2.0923>
   rotation=<rot 2.9368 16.047 -100.1001>
  }]
 foot_l [{shapeType=Capsule, length=11.0765, radius=8.0344, center=<-4.1811 6.0027 -0.338>, rotation=<rot 0.0388 1.4701 -89.3982>}]
 foot_r [{shapeType=Capsule, length=11.0765, radius=8.0344, center=<4.1811 -6.0027 0.338>, rotation=<rot 0.0388 1.4701 90.6018>}]
 head [{shapeType=Capsule, length=0, radius=18, center=<9.0697 0.0746 0>, rotation=<rot 0 -90 0>}]
 lowerarm_l []
 lowerarm_r []
 neck_02 []
 pelvis [{shapeType=Capsule, length=9.6042, radius=11.1441, center=<-2.4471 -2.5289 -0.1981>}]
 spine_01 [{shapeType=Capsule, length=4.3888, radius=14.3638, center=<-8.9021 -0.8973 0>}]
 spine_02 [
  {shapeType=Capsule, length=2.5219, radius=9.7317, center=<-2.6885 -1.2641 0>}
  {shapeType=Capsule, length=20.6875, radius=9.3125, center=<-1.8317 1.9362 0>, rotation=<rot 90 0 0>}
 ]
 spine_03 [
  {shapeType=Capsule, length=6.1863, radius=8.4854, center=<0.0574 0.4416 0>}
  {shapeType=Capsule, length=20.7188, radius=9.2812, center=<-1.2163 0.549 0>, rotation=<rot 90 180 -175.0732>}
 ]
 spine_04 [
  {shapeType=Capsule, length=9.0076, radius=10.3916, center=<5.6234 -0.6779 0>}
  {shapeType=Capsule, length=10.0312, radius=9.2812, center=<-1.2326 1.308 0>, rotation=<rot 90 0 0>}
 ]
 spine_05 [{shapeType=Capsule, length=9.0551, radius=12.9799, center=<0.6969 1.91 0>}]
 thigh_l []
 thigh_r []
 upperarm_l []
 upperarm_r []
}
constraints [
 {
  bone1Name=spine_01
  bone2Name=pelvis
  swing1LimitDegrees=45
  swing1Motion=Free
  swing2LimitDegrees=45
  swing2Motion=Free
  twistLimitDegrees=45
  twistMotion=Free
 }
 {
  bone1Name=spine_02
  bone2Name=spine_01
  swing1LimitDegrees=45
  swing1Motion=Free
  swing2LimitDegrees=45
  swing2Motion=Free
  twistLimitDegrees=45
  twistMotion=Free
 }
 {
  bone1Name=spine_03
  bone2Name=spine_02
  swing1LimitDegrees=45
  swing1Motion=Free
  swing2LimitDegrees=45
  swing2Motion=Free
  twistLimitDegrees=45
  twistMotion=Free
 }
 {
  bone1Name=spine_04
  bone2Name=spine_03
  swing1LimitDegrees=45
  swing1Motion=Free
  swing2LimitDegrees=45
  swing2Motion=Free
  twistLimitDegrees=45
  twistMotion=Free
 }
 {
  bone1Name=spine_05
  bone2Name=spine_04
  swing1LimitDegrees=45
  swing1Motion=Free
  swing2LimitDegrees=45
  swing2Motion=Free
  twistLimitDegrees=45
  twistMotion=Free
 }
 {
  bone1Name=clavicle_l
  bone2Name=spine_05
  swing1LimitDegrees=45
  swing1Motion=Free
  swing2LimitDegrees=45
  swing2Motion=Free
  twistLimitDegrees=45
  twistMotion=Free
 }
 {
  bone1Name=upperarm_l
  bone2Name=clavicle_l
  swing1LimitDegrees=45
  swing1Motion=Free
  swing2LimitDegrees=45
  swing2Motion=Free
  twistLimitDegrees=45
  twistMotion=Free
 }
 {
  bone1Name=lowerarm_l
  bone2Name=upperarm_l
  swing1LimitDegrees=45
  swing1Motion=Free
  swing2LimitDegrees=45
  swing2Motion=Free
  twistLimitDegrees=45
  twistMotion=Free
 }
 {
  bone1Name=neck_02
  bone2Name=spine_05
  swing1LimitDegrees=45
  swing1Motion=Free
  swing2LimitDegrees=45
  swing2Motion=Free
  twistLimitDegrees=45
  twistMotion=Free
 }
 {
  bone1Name=head
  bone2Name=neck_02
  swing1LimitDegrees=45
  swing1Motion=Free
  swing2LimitDegrees=45
  swing2Motion=Free
  twistLimitDegrees=45
  twistMotion=Free
 }
 {
  bone1Name=thigh_l
  bone2Name=pelvis
  swing1LimitDegrees=45
  swing1Motion=Free
  swing2LimitDegrees=45
  swing2Motion=Free
  twistLimitDegrees=45
  twistMotion=Free
 }
 {
  bone1Name=foot_l
  bone2Name=calf_l
  swing1LimitDegrees=45
  swing1Motion=Free
  swing2LimitDegrees=45
  swing2Motion=Free
  twistLimitDegrees=45
  twistMotion=Free
 }
 {
  bone1Name=lowerarm_r
  bone2Name=upperarm_r
  swing1LimitDegrees=45
  swing1Motion=Free
  swing2LimitDegrees=45
  swing2Motion=Free
  twistLimitDegrees=45
  twistMotion=Free
 }
 {
  bone1Name=upperarm_r
  bone2Name=clavicle_r
  swing1LimitDegrees=45
  swing1Motion=Free
  swing2LimitDegrees=45
  swing2Motion=Free
  twistLimitDegrees=45
  twistMotion=Free
 }
 {
  bone1Name=clavicle_r
  bone2Name=spine_05
  swing1LimitDegrees=45
  swing1Motion=Free
  swing2LimitDegrees=45
  swing2Motion=Free
  twistLimitDegrees=45
  twistMotion=Free
 }
 {
  bone1Name=thigh_r
  bone2Name=pelvis
  swing1LimitDegrees=45
  swing1Motion=Free
  swing2LimitDegrees=45
  swing2Motion=Free
  twistLimitDegrees=45
  twistMotion=Free
 }
 {
  bone1Name=foot_r
  bone2Name=calf_r
  swing1LimitDegrees=45
  swing1Motion=Free
  swing2LimitDegrees=45
  swing2Motion=Free
  twistLimitDegrees=45
  twistMotion=Free
 }
 {
  bone1Name=calf_r
  bone2Name=thigh_r
  swing1LimitDegrees=45
  swing1Motion=Free
  swing2LimitDegrees=45
  swing2Motion=Free
  twistLimitDegrees=45
  twistMotion=Free
 }
 {
  bone1Name=calf_l
  bone2Name=thigh_l
  swing1LimitDegrees=45
  swing1Motion=Free
  swing2LimitDegrees=45
  swing2Motion=Free
  twistLimitDegrees=45
  twistMotion=Free
 }
 {
  bone1Name=dyn_Skirt_01_F_l
  bone2Name=spine_01
  swing1LimitDegrees=45
  swing1Motion=Free
  swing2LimitDegrees=45
  swing2Motion=Free
  twistLimitDegrees=45
  twistMotion=Free
 }
 {
  bone1Name=dyn_Skirt_02_F_l
  bone2Name=dyn_Skirt_01_F_l
  swing1LimitDegrees=45
  swing1Motion=Free
  swing2LimitDegrees=45
  swing2Motion=Free
  twistLimitDegrees=45
  twistMotion=Free
 }
 {
  bone1Name=dyn_Skirt_03_F_l
  bone2Name=dyn_Skirt_02_F_l
  swing1LimitDegrees=45
  swing1Motion=Free
  swing2LimitDegrees=45
  swing2Motion=Free
  twistLimitDegrees=45
  twistMotion=Free
 }
 {
  bone1Name=dyn_Skirt_04_F_l
  bone2Name=dyn_Skirt_03_F_l
  swing1LimitDegrees=45
  swing1Motion=Free
  swing2LimitDegrees=45
  swing2Motion=Free
  twistLimitDegrees=45
  twistMotion=Free
 }
 {
  bone1Name=dyn_Skirt_05_F_l
  bone2Name=dyn_Skirt_04_F_l
  swing1LimitDegrees=45
  swing1Motion=Free
  swing2LimitDegrees=45
  swing2Motion=Free
  twistLimitDegrees=45
  twistMotion=Free
 }
 {
  bone1Name=dyn_Skirt_01_F_r
  bone2Name=spine_01
  swing1LimitDegrees=45
  swing1Motion=Free
  swing2LimitDegrees=45
  swing2Motion=Free
  twistLimitDegrees=45
  twistMotion=Free
 }
 {
  bone1Name=dyn_Skirt_02_F_r
  bone2Name=dyn_Skirt_01_F_r
  swing1LimitDegrees=45
  swing1Motion=Free
  swing2LimitDegrees=45
  swing2Motion=Free
  twistLimitDegrees=45
  twistMotion=Free
 }
 {
  bone1Name=dyn_Skirt_03_F_r
  bone2Name=dyn_Skirt_02_F_r
  swing1LimitDegrees=45
  swing1Motion=Free
  swing2LimitDegrees=45
  swing2Motion=Free
  twistLimitDegrees=45
  twistMotion=Free
 }
 {
  bone1Name=dyn_Skirt_04_F_r
  bone2Name=dyn_Skirt_03_F_r
  swing1LimitDegrees=45
  swing1Motion=Free
  swing2LimitDegrees=45
  swing2Motion=Free
  twistLimitDegrees=45
  twistMotion=Free
 }
 {
  bone1Name=dyn_Skirt_05_F_r
  bone2Name=dyn_Skirt_04_F_r
  swing1LimitDegrees=45
  swing1Motion=Free
  swing2LimitDegrees=45
  swing2Motion=Free
  twistLimitDegrees=45
  twistMotion=Free
 }
 {
  bone1Name=dyn_Skirt_02_SB_l
  bone2Name=dyn_Skirt_01_SB_l
  swing1LimitDegrees=45
  swing1Motion=Free
  swing2LimitDegrees=45
  swing2Motion=Free
  twistLimitDegrees=45
  twistMotion=Free
 }
 {
  bone1Name=dyn_Skirt_03_SB_l
  bone2Name=dyn_Skirt_02_SB_l
  swing1LimitDegrees=45
  swing1Motion=Free
  swing2LimitDegrees=45
  swing2Motion=Free
  twistLimitDegrees=45
  twistMotion=Free
 }
 {
  bone1Name=dyn_Skirt_04_SB_l
  bone2Name=dyn_Skirt_03_SB_l
  swing1LimitDegrees=45
  swing1Motion=Free
  swing2LimitDegrees=45
  swing2Motion=Free
  twistLimitDegrees=45
  twistMotion=Free
 }
 {
  bone1Name=dyn_Skirt_05_SB_l
  bone2Name=dyn_Skirt_04_SB_l
  swing1LimitDegrees=45
  swing1Motion=Free
  swing2LimitDegrees=45
  swing2Motion=Free
  twistLimitDegrees=45
  twistMotion=Free
 }
 {
  bone1Name=dyn_SkirtStrap_06_SB_l
  bone2Name=dyn_Skirt_05_SB_l
  swing1LimitDegrees=45
  swing1Motion=Free
  swing2LimitDegrees=45
  swing2Motion=Free
  twistLimitDegrees=45
  twistMotion=Free
 }
 {
  bone1Name=dyn_SkirtStrap_07_SB_l
  bone2Name=dyn_SkirtStrap_06_SB_l
  swing1LimitDegrees=45
  swing1Motion=Free
  swing2LimitDegrees=45
  swing2Motion=Free
  twistLimitDegrees=45
  twistMotion=Free
 }
 {
  bone1Name=dyn_Skirt_01_SB_r
  bone2Name=spine_01
  swing1LimitDegrees=45
  swing1Motion=Free
  swing2LimitDegrees=45
  swing2Motion=Free
  twistLimitDegrees=45
  twistMotion=Free
 }
 {
  bone1Name=dyn_Skirt_02_SB_r
  bone2Name=dyn_Skirt_01_SB_r
  swing1LimitDegrees=45
  swing1Motion=Free
  swing2LimitDegrees=45
  swing2Motion=Free
  twistLimitDegrees=45
  twistMotion=Free
 }
 {
  bone1Name=dyn_Skirt_03_SB_r
  bone2Name=dyn_Skirt_02_SB_r
  swing1LimitDegrees=45
  swing1Motion=Free
  swing2LimitDegrees=45
  swing2Motion=Free
  twistLimitDegrees=45
  twistMotion=Free
 }
 {
  bone1Name=dyn_Skirt_04_SB_r
  bone2Name=dyn_Skirt_03_SB_r
  swing1LimitDegrees=45
  swing1Motion=Free
  swing2LimitDegrees=45
  swing2Motion=Free
  twistLimitDegrees=45
  twistMotion=Free
 }
 {
  bone1Name=dyn_Skirt_05_SB_r
  bone2Name=dyn_Skirt_04_SB_r
  swing1LimitDegrees=45
  swing1Motion=Free
  swing2LimitDegrees=45
  swing2Motion=Free
  twistLimitDegrees=45
  twistMotion=Free
 }
 {
  bone1Name=dyn_SkirtStrap_06_SB_r
  bone2Name=dyn_Skirt_05_SB_r
  swing1LimitDegrees=45
  swing1Motion=Free
  swing2LimitDegrees=45
  swing2Motion=Free
  twistLimitDegrees=45
  twistMotion=Free
 }
 {
  bone1Name=dyn_SkirtStrap_07_SB_r
  bone2Name=dyn_SkirtStrap_06_SB_r
  swing1LimitDegrees=45
  swing1Motion=Free
  swing2LimitDegrees=45
  swing2Motion=Free
  twistLimitDegrees=45
  twistMotion=Free
 }
 {
  bone1Name=dyn_Skirt_01_S_l
  bone2Name=spine_01
  swing1LimitDegrees=45
  swing1Motion=Free
  swing2LimitDegrees=45
  swing2Motion=Free
  twistLimitDegrees=45
  twistMotion=Free
 }
 {
  bone1Name=dyn_Skirt_02_S_l
  bone2Name=dyn_Skirt_01_S_l
  swing1LimitDegrees=45
  swing1Motion=Free
  swing2LimitDegrees=45
  swing2Motion=Free
  twistLimitDegrees=45
  twistMotion=Free
 }
 {
  bone1Name=dyn_Skirt_03_S_l
  bone2Name=dyn_Skirt_02_S_l
  swing1LimitDegrees=45
  swing1Motion=Free
  swing2LimitDegrees=45
  swing2Motion=Free
  twistLimitDegrees=45
  twistMotion=Free
 }
 {
  bone1Name=dyn_Skirt_04_S_l
  bone2Name=dyn_Skirt_03_S_l
  swing1LimitDegrees=45
  swing1Motion=Free
  swing2LimitDegrees=45
  swing2Motion=Free
  twistLimitDegrees=45
  twistMotion=Free
 }
 {
  bone1Name=dyn_Skirt_05_S_l
  bone2Name=dyn_Skirt_04_S_l
  swing1LimitDegrees=45
  swing1Motion=Free
  swing2LimitDegrees=45
  swing2Motion=Free
  twistLimitDegrees=45
  twistMotion=Free
 }
 {
  bone1Name=dyn_Skirt_01_S_r
  bone2Name=spine_01
  swing1LimitDegrees=45
  swing1Motion=Free
  swing2LimitDegrees=45
  swing2Motion=Free
  twistLimitDegrees=45
  twistMotion=Free
 }
 {
  bone1Name=dyn_Skirt_02_S_r
  bone2Name=dyn_Skirt_01_S_r
  swing1LimitDegrees=45
  swing1Motion=Free
  swing2LimitDegrees=45
  swing2Motion=Free
  twistLimitDegrees=45
  twistMotion=Free
 }
 {
  bone1Name=dyn_Skirt_03_S_r
  bone2Name=dyn_Skirt_02_S_r
  swing1LimitDegrees=45
  swing1Motion=Free
  swing2LimitDegrees=45
  swing2Motion=Free
  twistLimitDegrees=45
  twistMotion=Free
 }
 {
  bone1Name=dyn_Skirt_04_S_r
  bone2Name=dyn_Skirt_03_S_r
  swing1LimitDegrees=45
  swing1Motion=Free
  swing2LimitDegrees=45
  swing2Motion=Free
  twistLimitDegrees=45
  twistMotion=Free
 }
 {
  bone1Name=dyn_Skirt_05_S_r
  bone2Name=dyn_Skirt_04_S_r
  swing1LimitDegrees=45
  swing1Motion=Free
  swing2LimitDegrees=45
  swing2Motion=Free
  twistLimitDegrees=45
  twistMotion=Free
 }
 {
  bone1Name=dyn_Skirt_01_SB_l
  bone2Name=spine_01
  swing1LimitDegrees=45
  swing1Motion=Free
  swing2LimitDegrees=45
  swing2Motion=Free
  twistLimitDegrees=45
  twistMotion=Free
 }
]
