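# Post-process Control Rig for the F_MED_GraveFlannel character body that corrects its bracelet chain, hanging key and skirt bones.
# It locks bracelet translations, runs skirt fixes for sky floating, low FPS, front-end idle and skydiving when RBAN is off, and always runs the key corrector.
# Additive bone offsets are weighted by state variables, swing-twist rotation deltas and pelvis-height curves for crouching, with many weights interpolated for smooth blending.
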ControlRigBlueprint /BRCosmetics/Characters/Player/Female/Medium/Bodies/F_MED_GraveFlannel/Meshes/F_MED_GraveFlannel_Post_CtrlRig
graphs [
 Multiply_1_ContainedGraph {nodes [
   n1 = RigVMFunctionEntryNode()
   Multiply = &RigVMFunction_MathDoubleMul(A=@n1.A, B=@n1.B)
   RigVMFunctionReturnNode(Result=&RigVMFunction_MathDoubleMul(B=@n1.C, A=@Multiply.Result).Result)
  ]}
 New Function_1_ContainedGraph {nodes [
   n1 = RigVMFunctionEntryNode()
   n2 = &RigUnit_GetTransform(Space=LocalSpace, bInitial=False, Item=@n1.Item)
   RigVMFunction_MathQuaternionToEuler = &RigVMFunction_MathQuaternionToEuler(RotationOrder=ZYX, Value=@n2.Transform.Rotation)
   Clamp = &RigVMFunction_MathIntClamp(Value=@n1.Rotate Index)
   n3 = RigVMDispatchNode(
    Values=(0,0,0)
    Index=@Clamp.Result
    Values.0=@RigVMFunction_MathQuaternionToEuler.Result.X
    Values.1=@RigVMFunction_MathQuaternionToEuler.Result.Y
    Values.2=@RigVMFunction_MathQuaternionToEuler.Result.Z
   ) {template="DISPATCH_RigVMDispatch_SelectInt32(in Index,in Values,out Result)"}
   n4 = RigVMVariableNode() {variable=Max Rotate}
   LessEqual = /Script/RigVM.RigVMFunction_MathDoubleLessEqual(A=@n3.Result, B=RigVMVariableNode() {variable=Max Rotate}.Value) {is=struct RigVMUnitNode, template="LessEqual::Execute(in A,in B,out Result)"}
   RigVMFunction_MathBoolToInteger = &RigVMFunction_MathBoolToInteger(Value=RigVMDispatchNode(
     Values=(False,False)
     Values.0=/Script/RigVM.RigVMFunction_MathDoubleGreaterEqual(A=@n3.Result, B=@n4.Value) {is=struct RigVMUnitNode, template="GreaterEqual::Execute(in A,in B,out Result)"}.Result
     Values.1=@LessEqual.Result
     Index=&RigVMFunction_MathBoolToInteger(Value=@n1.Is Lower limit).Result
    ) {template="DISPATCH_RigVMDispatch_SelectInt32(in Index,in Values,out Result)"}.Result)
   n5 = RigVMDispatchNode(Values=(0,0), Index=@RigVMFunction_MathBoolToInteger.Result, Values.0=@n3.Result, Values.1=RigVMVariableNode() {variable=Max Rotate}.Value) {template="DISPATCH_RigVMDispatch_SelectInt32(in Index,in Values,out Result)"}
   n6 = RigVMDispatchNode(
    Values=((X=0,Y=0,Z=0),(X=0,Y=0,Z=0),(X=0,Y=0,Z=0))
    Values.1.X=@RigVMFunction_MathQuaternionToEuler.Result.X
    Values.2.X=@RigVMFunction_MathQuaternionToEuler.Result.X
    Values.0.Y=@RigVMFunction_MathQuaternionToEuler.Result.Y
    Values.2.Y=@RigVMFunction_MathQuaternionToEuler.Result.Y
    Values.0.Z=@RigVMFunction_MathQuaternionToEuler.Result.Z
    Values.1.Z=@RigVMFunction_MathQuaternionToEuler.Result.Z
    Values.0.X=@n5.Result
    Values.1.Y=@n5.Result
    Values.2.Z=@n5.Result
    Index=@Clamp.Result
   ) {template="DISPATCH_RigVMDispatch_SelectInt32(in Index,in Values,out Result)"}
   > RigVMDispatchNode(Enabled=false, Prefix=Rotation Values, ScreenColor=(R=1,G=1,B=1,A=1), ScreenDuration=0.05, Value=@n6.Result) {template="DISPATCH_RigVMDispatch_Print(in Prefix,in Value,in Enabled,in ScreenDuration,in ScreenColor)"}
   RigVMFunction_MathQuaternionFromEuler = &RigVMFunction_MathQuaternionFromEuler(RotationOrder=XYZ, Euler=@n6.Result)
   Interpolate = /Script/RigVM.RigVMFunction_MathQuaternionSlerp(A=@RigVMFunction_MathQuaternionFromEuler.Result, B=@n2.Transform.Rotation, T=@n1.Alpha) {is=struct RigVMUnitNode, template="Interpolate::Execute(in A,in B,in T,out Result)"}
   > Set Transform = &RigUnit_SetRotation(Weight=1, bInitial=False, Item=@n1.Item, Space=@n1.Space, Value=@Interpolate.Result)
   > RigVMFunctionReturnNode()
  ]}
 New Function_ContainedGraph {nodes [
   DeltaRotation v4_1 = RigVMCollapseNode(
    ExecuteContext=()
    Rotation Order=ZYX
    child=(Type=Bone,Name="spine_04")
    debugRotation=false
    outputMinMax=(X=0,Y=1)
    parent=(Type=Null,Name="spine_04_PevlisDeltaRef_null")
    premult=1
    rotationMinMax=(X=12,Y=-1)
    triggerAxis=Z
    twistAxis=(X=1,Y=0,Z=0)
   )
   > &RigUnit_ModifyTransforms(
    ItemToModify="((Item=(Type=Bone,Name=\"base_key_Start_l\"),Transform=(Rotation=(X=0.0435,Y=-0.0038,Z=-0.0871,W=0.9952),Translation=(X=0,Y=0,Z=0),Scale3D=(X=1,Y=1,Z=1))),(Item=(Type=Bone,Name=\"base_key_l\"),Transform=(Rotation=(X=0.2588,Y=0,Z=0,W=0.9659),Translation=(X=0,Y= ...(truncated)"
    WeightMinimum=0
    Weight=@"DeltaRotation v4_1".outputWeight
   )
   > DeltaRotation v4_2 = RigVMCollapseNode(
    Rotation Order=ZYX
    absoluteValue=false
    child=(Type=Bone,Name="spine_05")
    debugRotation=false
    outputMinMax=(X=0,Y=1)
    parent=(Type=Null,Name="spine_05_PevlisDeltaRef_null")
    premult=1
    rotationMinMax=(X=0,Y=1)
    triggerAxis=Y
    twistAxis=(X=1,Y=0,Z=0)
   )
   > &RigUnit_ModifyTransforms(
    ItemToModify="((Item=(Type=Bone,Name=\"base_key_Start_l\"),Transform=(Rotation=(X=0.0436,Y=0,Z=0,W=0.999),Translation=(X=0,Y=0,Z=0),Scale3D=(X=1,Y=1,Z=1))),(Item=(Type=Bone,Name=\"base_key_l\"),Transform=(Rotation=(X=0.1736,Y=0,Z=0,W=0.9848),Translation=(X=0,Y=0,Z=0),Scale3 ...(truncated)"
    WeightMinimum=0
    Weight=@"DeltaRotation v4_2".outputWeight
   )
   RigVMFunctionEntryNode()
   Subtract = &RigVMFunction_MathDoubleSub(A=1, B=RigVMVariableNode() {variable=isCharacterIdle}.Value)
   AlphaInterp = &RigVMFunction_AlphaInterp(InterpSpeedDecreasing=6, InterpSpeedIncreasing=3, Scale=1, bClampResult=False, Value=@Subtract.Result)
   > &RigUnit_ModifyTransforms(
    ItemToModify="((Item=(Type=Bone,Name=\"base_key_Start_l\"),Transform=(Rotation=(X=0.0002,Y=-0.0262,Z=0.0087,W=0.9996),Translation=(X=0,Y=0,Z=0.2),Scale3D=(X=1,Y=1,Z=1))),(Item=(Type=Bone,Name=\"base_key_l\"),Transform=(Rotation=(X=-0.0259,Y=-0.0264,Z=0.008,W=0.9993),Transla ...(truncated)"
    WeightMinimum=0
    Weight=@AlphaInterp.Result
   )
   > DeltaRotation v4 = RigVMCollapseNode(
    Rotation Order=ZYX
    child=(Type=Bone,Name="dyn_Skirt_01_SB_l")
    debugRotation=false
    outputMinMax=(X=0,Y=1)
    parent=(Type=Null,Name="dyn_Skirt_01_SB_l_Spine01DeltaRef_null")
    premult=1
    rotationMinMax=(X=0,Y=-15)
    triggerAxis=Y
    twistAxis=(X=1,Y=0,Z=0)
   )
   > n1 = &RigUnit_ModifyTransforms(
    ItemToModify="((Item=(Type=Bone,Name=\"base_key_Start_l\"),Transform=(Rotation=(X=-0.1731,Y=-0.0253,Z=0.0177,W=0.9844),Translation=(X=0,Y=0,Z=0),Scale3D=(X=1,Y=1,Z=1))),(Item=(Type=Bone,Name=\"base_key_l\"),Transform=(Rotation=(X=-0.0649,Y=-0.0232,Z=0.0203,W=0.9974),Transla ...(truncated)"
    WeightMinimum=0
    Weight=@"DeltaRotation v4".outputWeight
   )
   RigVMFunction_ControlFlowBranch = &RigVMFunction_ControlFlowBranch(ExecuteContext=@n1.ExecutePin, Condition=RigVMVariableNode() {variable=isRbanEnabled}.Value)
   &RigUnit_ModifyTransforms(
    ItemToModify="((Item=(Type=Bone,Name=\"base_key_Start_l\"),Transform=(Rotation=(X=0.2495,Y=0.1063,Z=0.0614,W=0.9606),Translation=(X=0,Y=0,Z=0),Scale3D=(X=1,Y=1,Z=1))),(Item=(Type=Bone,Name=\"base_key_l\"),Transform=(Rotation=(X=0.421,Y=-0.0368,Z=-0.079,W=0.9029),Translation ...(truncated)"
    WeightMinimum=0
    ExecutePin=@RigVMFunction_ControlFlowBranch.Completed
    Weight=RigVMVariableNode() {variable=isBaseballSliding}.Value
   )
   > &RigUnit_ModifyTransforms(
    ItemToModify="((Item=(Type=Bone,Name=\"base_key_Start_l\"),Transform=(Rotation=(X=0,Y=0.0872,Z=0,W=0.9962),Translation=(X=0,Y=0,Z=0),Scale3D=(X=1,Y=1,Z=1))),(Item=(Type=Bone,Name=\"base_key_l\"),Transform=(Rotation=(X=0.2666,Y=0.1032,Z=-0.1171,W=0.9511),Translation=(X=0,Y=0 ...(truncated)"
    WeightMinimum=0
    Weight=RigVMVariableNode() {variable=isKneeSliding}.Value
   )
   > RigVMFunctionReturnNode()
   RigVMFunction_MathBoolToFloat = &RigVMFunction_MathBoolToFloat(Value=RigVMVariableNode() {variable=isSkyFloating}.Value)
   n2 = &RigVMFunction_AlphaInterp(InterpSpeedDecreasing=10, InterpSpeedIncreasing=10, Scale=1, bClampResult=False, Value=@RigVMFunction_MathBoolToFloat.Result)
   &RigUnit_ModifyTransforms(
    ItemToModify=((Item=(Type=Bone,Name="base_key_l"),Transform=(Rotation=(X=-0.4226,Y=0,Z=0,W=0.9063),Translation=(X=0,Y=0,Z=0),Scale3D=(X=1,Y=1,Z=1))))
    WeightMinimum=0
    Weight=@n2.Result
    ExecutePin=@RigVMFunction_ControlFlowBranch.False
   )
   n3 = &RigVMFunction_AnimEvalRichCurve(
    Curve=(EditorCurveData=(Keys=((InterpMode=RCIM_Cubic),(InterpMode=RCIM_Cubic,Time=0.5,Value=1),(Time=0.8))))
    SourceMaximum=2
    SourceMinimum=0
    TargetMinimum=0
    Value=&RigVMFunction_MathDoubleRemap(
     SourceMaximum=60
     SourceMinimum=45
     TargetMaximum=0
     TargetMinimum=1
     bClamp=false
     Value=&RigUnit_GetRelativeTransformForItem(Child=(Type=Bone,Name="pelvis"), Parent=(Type=Bone,Name="root"), bChildInitial=False, bParentInitial=False).RelativeTransform.Translation.Z
    ).Result
   )
   &RigUnit_ModifyTransforms(
    ItemToModify="((Item=(Type=Bone,Name=\"base_key_Start_l\"),Transform=(Rotation=(X=0.0871,Y=-0.0435,Z=0.0038,W=0.9952),Translation=(X=0,Y=0,Z=0),Scale3D=(X=1,Y=1,Z=1))),(Item=(Type=Bone,Name=\"base_key_l\"),Transform=(Rotation=(X=0.0883,Y=0.0097,Z=-0.0883,W=0.9921),Translati ...(truncated)"
    WeightMinimum=0
    Weight=@n3.Result
    ExecutePin=@RigVMFunction_ControlFlowBranch.True
   )
   n4 = &RigVMFunction_AnimEvalRichCurve(
    Curve=(EditorCurveData=(Keys=((InterpMode=RCIM_Cubic),(InterpMode=RCIM_Cubic,Time=0.5,Value=1),(Time=0.8))))
    SourceMaximum=2
    SourceMinimum=0
    TargetMinimum=0
    Value=&RigVMFunction_MathDoubleRemap(
     SourceMaximum=75
     SourceMinimum=45
     TargetMaximum=1
     TargetMinimum=0
     bClamp=false
     Value=&RigUnit_GetRelativeTransformForItem(Child=(Type=Bone,Name="pelvis"), Parent=(Type=Bone,Name="root"), bChildInitial=False, bParentInitial=False).RelativeTransform.Translation.Z
    ).Result
   )
   > RigVMVariableNode(Value=@n4.Result) {variable=CrouchWalking}
   > &RigUnit_ModifyTransforms(
    ItemToModify="((Item=(Type=Bone,Name=\"base_key_Start_l\"),Transform=(Rotation=(X=-0.0871,Y=-0.0435,Z=-0.0038,W=0.9952),Translation=(X=0,Y=0,Z=0),Scale3D=(X=1,Y=1,Z=1))),(Item=(Type=Bone,Name=\"base_key_l\"),Transform=(Rotation=(X=-0.0246,Y=0.0197,Z=-0.0867,W=0.9957),Transl ...(truncated)"
    WeightMinimum=0
    Weight=@n4.Result
   )
   n5 = &RigVMFunction_AlphaInterp(InterpSpeedDecreasing=5, InterpSpeedIncreasing=5, Scale=1, bClampResult=False, Value=RigVMVariableNode() {variable=MovingBack}.Value)
   Multiply = &RigVMFunction_MathDoubleMul(A=@n5.Result, B=&RigVMFunction_MathDoubleSub(A=1, B=RigVMVariableNode() {variable=CrouchWalking}.Value).Result)
   > &RigUnit_ModifyTransforms(
    ItemToModify=((Item=(Type=Bone,Name="base_key_Start_l"),Transform=(Rotation=(X=0.1736,Y=-0.003,Z=-0.0172,W=0.9847),Translation=(X=0,Y=0,Z=0),Scale3D=(X=1,Y=1,Z=1))))
    WeightMinimum=0
    Weight=@Multiply.Result
   )
  ]}
 New Function_ContainedGraph {nodes [
   n1 = RigVMFunctionEntryNode()
   RigVMFunction_MathBoolToInteger = &RigVMFunction_MathBoolToInteger(Value=@n1.onlyTwist)
   RigVMFunction_MathQuaternionSwingTwist = &RigVMFunction_MathQuaternionSwingTwist(
    TwistAxis=@n1.twistAxis
    Input=&RigUnit_GetRelativeTransformForItem(bChildInitial=False, bParentInitial=False, Child=@n1.child, Parent=@n1.parent).RelativeTransform.Rotation
   )
   n2 = RigVMDispatchNode(
    Values=((),())
    Index=@RigVMFunction_MathBoolToInteger.Result
    Values.0=@RigVMFunction_MathQuaternionSwingTwist.Swing
    Values.1=@RigVMFunction_MathQuaternionSwingTwist.Twist
   ) {template="DISPATCH_RigVMDispatch_SelectInt32(in Index,in Values,out Result)"}
   RigVMFunction_MathQuaternionToEuler = &RigVMFunction_MathQuaternionToEuler(Value=@n2.Result, RotationOrder=@n1.Rotation Order)
   > DISPATCH_RigVMDispatch_Print = RigVMDispatchNode(ScreenColor=(R=1,G=1,B=1,A=1), ScreenDuration=1.5, Enabled=@n1.debugRotation, Value=@RigVMFunction_MathQuaternionToEuler.Result) {template="DISPATCH_RigVMDispatch_Print(in Prefix,in Value,in Enabled,in ScreenDuration,in ScreenColor)"}
   Cast = RigVMDispatchNode(Value=@n1.triggerAxis) {template="DISPATCH_RigVMDispatch_CastEnumToInt(in Value,out Result)"}
   DISPATCH_RigVMDispatch_SelectInt32 = RigVMDispatchNode(
    Values=(0,0,0)
    Index=@Cast.Result
    Values.0=@RigVMFunction_MathQuaternionToEuler.Result.X
    Values.1=@RigVMFunction_MathQuaternionToEuler.Result.Y
    Values.2=@RigVMFunction_MathQuaternionToEuler.Result.Z
   ) {template="DISPATCH_RigVMDispatch_SelectInt32(in Index,in Values,out Result)"}
   Remap = &RigVMFunction_MathDoubleRemap(
    bClamp=False
    Value=@DISPATCH_RigVMDispatch_SelectInt32.Result
    TargetMinimum=@n1.outputMinMax.X
    TargetMaximum=@n1.outputMinMax.Y
    SourceMinimum=@n1.rotationMinMax.X
    SourceMaximum=@n1.rotationMinMax.Y
   )
   Absolute = &RigVMFunction_MathDoubleAbs(Value=@Remap.Result)
   Multiply = &RigVMFunction_MathDoubleMul(
    A=&RigVMFunction_MathDoubleClamp(Value=RigVMDispatchNode(True=@Absolute.Result, Condition=@n1.absoluteValue, False=@Remap.Result) {template="DISPATCH_RigVMDispatch_If(in Condition,in True,in False,out Result)"}.Result).Result
    B=@n1.premult
   )
   > RigVMFunctionReturnNode(outputQuat=@n2.Result, outputWeight=@Multiply.Result)
  ]}
 New Function_ContainedGraph {nodes [
   n1 = RigVMFunctionEntryNode()
   RigVMFunction_MathBoolToInteger = &RigVMFunction_MathBoolToInteger(Value=@n1.onlyTwist)
   RigVMFunction_MathQuaternionSwingTwist = &RigVMFunction_MathQuaternionSwingTwist(
    TwistAxis=@n1.twistAxis
    Input=&RigUnit_GetRelativeTransformForItem(bChildInitial=False, bParentInitial=False, Child=@n1.child, Parent=@n1.parent).RelativeTransform.Rotation
   )
   n2 = RigVMDispatchNode(
    Values=((),())
    Index=@RigVMFunction_MathBoolToInteger.Result
    Values.0=@RigVMFunction_MathQuaternionSwingTwist.Swing
    Values.1=@RigVMFunction_MathQuaternionSwingTwist.Twist
   ) {template="DISPATCH_RigVMDispatch_SelectInt32(in Index,in Values,out Result)"}
   RigVMFunction_MathQuaternionToEuler = &RigVMFunction_MathQuaternionToEuler(Value=@n2.Result, RotationOrder=@n1.Rotation Order)
   > DISPATCH_RigVMDispatch_Print = RigVMDispatchNode(ScreenColor=(R=1,G=1,B=1,A=1), ScreenDuration=1.5, Enabled=@n1.debugRotation, Value=@RigVMFunction_MathQuaternionToEuler.Result) {template="DISPATCH_RigVMDispatch_Print(in Prefix,in Value,in Enabled,in ScreenDuration,in ScreenColor)"}
   Cast = RigVMDispatchNode(Value=@n1.triggerAxis) {template="DISPATCH_RigVMDispatch_CastEnumToInt(in Value,out Result)"}
   DISPATCH_RigVMDispatch_SelectInt32 = RigVMDispatchNode(
    Values=(0,0,0)
    Index=@Cast.Result
    Values.0=@RigVMFunction_MathQuaternionToEuler.Result.X
    Values.1=@RigVMFunction_MathQuaternionToEuler.Result.Y
    Values.2=@RigVMFunction_MathQuaternionToEuler.Result.Z
   ) {template="DISPATCH_RigVMDispatch_SelectInt32(in Index,in Values,out Result)"}
   Remap = &RigVMFunction_MathDoubleRemap(
    bClamp=False
    Value=@DISPATCH_RigVMDispatch_SelectInt32.Result
    TargetMinimum=@n1.outputMinMax.X
    TargetMaximum=@n1.outputMinMax.Y
    SourceMinimum=@n1.rotationMinMax.X
    SourceMaximum=@n1.rotationMinMax.Y
   )
   Absolute = &RigVMFunction_MathDoubleAbs(Value=@Remap.Result)
   Multiply = &RigVMFunction_MathDoubleMul(
    A=&RigVMFunction_MathDoubleClamp(Value=RigVMDispatchNode(True=@Absolute.Result, Condition=@n1.absoluteValue, False=@Remap.Result) {template="DISPATCH_RigVMDispatch_If(in Condition,in True,in False,out Result)"}.Result).Result
    B=@n1.premult
   )
   > RigVMFunctionReturnNode(outputQuat=@n2.Result, outputWeight=@Multiply.Result)
  ]}
 New Function_ContainedGraph {nodes [
   n1 = RigVMFunctionEntryNode()
   RigVMFunction_MathBoolToInteger = &RigVMFunction_MathBoolToInteger(Value=@n1.onlyTwist)
   RigVMFunction_MathQuaternionSwingTwist = &RigVMFunction_MathQuaternionSwingTwist(
    TwistAxis=@n1.twistAxis
    Input=&RigUnit_GetRelativeTransformForItem(bChildInitial=False, bParentInitial=False, Child=@n1.child, Parent=@n1.parent).RelativeTransform.Rotation
   )
   n2 = RigVMDispatchNode(
    Values=((),())
    Index=@RigVMFunction_MathBoolToInteger.Result
    Values.0=@RigVMFunction_MathQuaternionSwingTwist.Swing
    Values.1=@RigVMFunction_MathQuaternionSwingTwist.Twist
   ) {template="DISPATCH_RigVMDispatch_SelectInt32(in Index,in Values,out Result)"}
   RigVMFunction_MathQuaternionToEuler = &RigVMFunction_MathQuaternionToEuler(Value=@n2.Result, RotationOrder=@n1.Rotation Order)
   > DISPATCH_RigVMDispatch_Print = RigVMDispatchNode(ScreenColor=(R=1,G=1,B=1,A=1), ScreenDuration=1.5, Enabled=@n1.debugRotation, Value=@RigVMFunction_MathQuaternionToEuler.Result) {template="DISPATCH_RigVMDispatch_Print(in Prefix,in Value,in Enabled,in ScreenDuration,in ScreenColor)"}
   Cast = RigVMDispatchNode(Value=@n1.triggerAxis) {template="DISPATCH_RigVMDispatch_CastEnumToInt(in Value,out Result)"}
   DISPATCH_RigVMDispatch_SelectInt32 = RigVMDispatchNode(
    Values=(0,0,0)
    Index=@Cast.Result
    Values.0=@RigVMFunction_MathQuaternionToEuler.Result.X
    Values.1=@RigVMFunction_MathQuaternionToEuler.Result.Y
    Values.2=@RigVMFunction_MathQuaternionToEuler.Result.Z
   ) {template="DISPATCH_RigVMDispatch_SelectInt32(in Index,in Values,out Result)"}
   Remap = &RigVMFunction_MathDoubleRemap(
    bClamp=False
    Value=@DISPATCH_RigVMDispatch_SelectInt32.Result
    TargetMinimum=@n1.outputMinMax.X
    TargetMaximum=@n1.outputMinMax.Y
    SourceMinimum=@n1.rotationMinMax.X
    SourceMaximum=@n1.rotationMinMax.Y
   )
   Absolute = &RigVMFunction_MathDoubleAbs(Value=@Remap.Result)
   Multiply = &RigVMFunction_MathDoubleMul(
    A=&RigVMFunction_MathDoubleClamp(Value=RigVMDispatchNode(True=@Absolute.Result, Condition=@n1.absoluteValue, False=@Remap.Result) {template="DISPATCH_RigVMDispatch_If(in Condition,in True,in False,out Result)"}.Result).Result
    B=@n1.premult
   )
   > RigVMFunctionReturnNode(outputQuat=@n2.Result, outputWeight=@Multiply.Result)
  ]}
 New Function_ContainedGraph {nodes [
   RigVMFunctionEntryNode()
   RigUnit_GetTransform = &RigUnit_GetTransform(Item=(Type=Bone,Name="dyn_wristChain_01_r"), Space=LocalSpace, bInitial=False)
   > &RigUnit_SetTransform(
    Item=(Type=Bone,Name="dyn_wristChain_01_r")
    Space=LocalSpace
    Value=(Rotation=(X=0,Y=0,Z=0,W=1),Translation=(X=0,Y=0,Z=0),Scale3D=(X=1,Y=1,Z=1))
    Weight=1
    bInitial=False
    Value.Rotation=@RigUnit_GetTransform.Transform.Rotation
    Value.Scale3D=@RigUnit_GetTransform.Transform.Scale3D
   )
   n1 = &RigUnit_GetTransform(Item=(Type=Bone,Name="dyn_wristChain_02_r"), Space=LocalSpace, bInitial=False)
   > &RigUnit_SetTransform(
    Item=(Type=Bone,Name="dyn_wristChain_02_r")
    Space=LocalSpace
    Value=(Rotation=(X=0,Y=0,Z=0,W=1),Translation=(X=1.1108,Y=0,Z=0),Scale3D=(X=1,Y=1,Z=1))
    Weight=1
    bInitial=False
    Value.Rotation=@n1.Transform.Rotation
    Value.Scale3D=@n1.Transform.Scale3D
   )
   n2 = &RigUnit_GetTransform(Item=(Type=Bone,Name="dyn_wristChain_03_r"), Space=LocalSpace, bInitial=False)
   > &RigUnit_SetTransform(
    Item=(Type=Bone,Name="dyn_wristChain_03_r")
    Space=LocalSpace
    Value=(Rotation=(X=0,Y=0,Z=0,W=1),Translation=(X=2.8056,Y=0,Z=0),Scale3D=(X=1,Y=1,Z=1))
    Weight=1
    bInitial=False
    Value.Rotation=@n2.Transform.Rotation
    Value.Scale3D=@n2.Transform.Scale3D
   )
   > RigVMFunctionReturnNode()
  ]}
 New Function_ContainedGraph {nodes [
   RigVMFunctionEntryNode()
   RigVMFunction_MathBoolToFloat = &RigVMFunction_MathBoolToFloat(Value=RigVMVariableNode() {variable=isSkyFloating}.Value)
   AlphaInterp = &RigVMFunction_AlphaInterp(InterpSpeedDecreasing=8, InterpSpeedIncreasing=8, Scale=1, bClampResult=False, Value=@RigVMFunction_MathBoolToFloat.Result)
   > &RigUnit_ModifyTransforms(
    ItemToModify=((Item=(Type=Bone,Name="dyn_Skirt_02_S_l"),Transform=(Rotation=(X=-0.0151,Y=-0.0858,Z=-0.173,W=0.9811),Translation=(X=0,Y=0,Z=0),Scale3D=(X=1,Y=1,Z=1))))
    WeightMinimum=0
    Weight=@AlphaInterp.Result
   )
   > RigVMFunctionReturnNode()
   LimitRotValue = RigVMCollapseNode(Alpha=1, ExecuteContext=(), Is Lower limit=true, Item=(Type=Bone,Name="dyn_Skirt_03_S_r"), Max Rotate=0, Rotate Index=1, Space=LocalSpace)
  ]}
 New Function_ContainedGraph {nodes [
   n1 = RigVMFunctionEntryNode()
   RigVMFunction_Sequence = &RigVMFunction_Sequence(ExecuteContext=@n1.ExecuteContext)
   RigUnit_GetTransform = &RigUnit_GetTransform(Item=(Type=Bone,Name="dyn_Skirt_05_S_r"), bInitial=False)
   RigVMFunction_MathVectorDistance = &RigVMFunction_MathVectorDistance(
    A=@RigUnit_GetTransform.Transform.Translation
    B=&RigUnit_GetTransform(Item=(Type=Bone,Name="dyn_Skirt_05_SB_r"), bInitial=False).Transform.Translation
   )
   Remap = &RigVMFunction_MathDoubleRemap(SourceMaximum=22, SourceMinimum=14, TargetMaximum=0, TargetMinimum=1, Value=@RigVMFunction_MathVectorDistance.Result)
   Multiply = RigVMAggregateNode(B=@Remap.Result, C=RigVMVariableNode() {variable=lowFPs}.Value, A=RigVMVariableNode() {variable=isOnlyCrouchingStill}.Value)
   &RigUnit_OffsetTransformForItem(
    Item=(Type=Bone,Name="dyn_Skirt_03_S_r")
    OffsetTransform=(Rotation=(X=0.1805,Y=-0.204,Z=-0.1299,W=0.9534),Translation=(X=0,Y=1,Z=1),Scale3D=(X=1,Y=1,Z=1))
    bPropagateToChildren=True
    Weight=@Multiply.Result
    ExecutePin=@RigVMFunction_Sequence.B
   )
   > RigVMFunctionReturnNode()
   n2 = &RigVMFunction_AnimEvalRichCurve(
    Curve=(EditorCurveData=(Keys=((InterpMode=RCIM_Cubic),(InterpMode=RCIM_Cubic,Time=0.5,Value=1),(Time=0.8))))
    SourceMaximum=2
    SourceMinimum=0
    TargetMinimum=0
    Value=&RigVMFunction_MathDoubleRemap(
     SourceMaximum=60
     SourceMinimum=45
     TargetMaximum=0
     TargetMinimum=1
     bClamp=false
     Value=&RigUnit_GetRelativeTransformForItem(Child=(Type=Bone,Name="pelvis"), Parent=(Type=Bone,Name="root"), bChildInitial=False, bParentInitial=False).RelativeTransform.Translation.Z
    ).Result
   )
   RigVMVariableNode(Value=@n2.Result, ExecuteContext=@RigVMFunction_Sequence.A) {variable=isOnlyCrouchingStill}
  ]}
 New Function_ContainedGraph {nodes [
   RigVMFunctionEntryNode()
   RigVMFunction_MathBoolToFloat = &RigVMFunction_MathBoolToFloat(Value=RigVMVariableNode() {variable=isRbanEnabled}.Value)
   Multiply = &RigVMFunction_MathDoubleMul(A=@RigVMFunction_MathBoolToFloat.Result, B=RigVMVariableNode() {variable=isIdleInFE}.Value)
   > &RigUnit_OffsetTransformForItem(
    Item=(Type=Bone,Name="dyn_Skirt_05_F_l")
    OffsetTransform=(Rotation=(X=0,Y=-0.1736,Z=0,W=0.9848),Translation=(X=0,Y=0,Z=-1),Scale3D=(X=1,Y=1,Z=1))
    bPropagateToChildren=false
    Weight=@Multiply.Result
   )
   > &RigUnit_OffsetTransformForItem(
    Item=(Type=Bone,Name="dyn_Skirt_01_F_l")
    OffsetTransform=(Rotation=(X=0,Y=0.0436,Z=0,W=0.999),Translation=(X=0,Y=0,Z=0),Scale3D=(X=1,Y=1,Z=1))
    bPropagateToChildren=false
    Weight=RigVMVariableNode() {variable=isIdleInFE}.Value
   )
   > &RigUnit_OffsetTransformForItem(
    Item=(Type=Bone,Name="dyn_Skirt_03_F_l")
    OffsetTransform=(Rotation=(X=0,Y=0,Z=0,W=1),Translation=(X=0,Y=0,Z=1.5),Scale3D=(X=1,Y=1,Z=1))
    bPropagateToChildren=false
    Weight=RigVMVariableNode() {variable=isIdleInFE}.Value
   )
   > &RigUnit_OffsetTransformForItem(
    Item=(Type=Bone,Name="dyn_Skirt_04_F_l")
    OffsetTransform=(Rotation=(X=0,Y=0.0436,Z=0,W=0.999),Translation=(X=0,Y=0,Z=0),Scale3D=(X=1,Y=1,Z=1))
    bPropagateToChildren=true
    Weight=RigVMVariableNode() {variable=isIdleInFE}.Value
   )
   > &RigUnit_OffsetTransformForItem(
    Item=(Type=Bone,Name="base_key_Start_l")
    OffsetTransform=(Rotation=(X=0.1351,Y=0.0489,Z=-0.0942,W=0.9851),Translation=(X=0,Y=0,Z=0),Scale3D=(X=1,Y=1,Z=1))
    bPropagateToChildren=true
    Weight=RigVMVariableNode() {variable=isIdleInFE}.Value
   )
   > &RigUnit_OffsetTransformForItem(
    Item=(Type=Bone,Name="base_key_l")
    OffsetTransform=(Rotation=(X=0.1305,Y=0,Z=0,W=0.9914),Translation=(X=0,Y=0,Z=0),Scale3D=(X=1,Y=1,Z=1))
    bPropagateToChildren=true
    Weight=RigVMVariableNode() {variable=isIdleInFE}.Value
   )
   > RigVMFunctionReturnNode()
  ]}
 New Function_ContainedGraph {nodes [
   RigVMFunctionEntryNode()
   RigVMFunction_MathBoolToFloat = &RigVMFunction_MathBoolToFloat(Value=RigVMVariableNode() {variable=isDiveMode}.Value)
   AlphaInterp = &RigVMFunction_AlphaInterp(InterpSpeedDecreasing=8, InterpSpeedIncreasing=10, Scale=1, bClampResult=False, Value=@RigVMFunction_MathBoolToFloat.Result)
   > &RigUnit_ModifyTransforms(
    ItemToModify="((Item=(Type=Bone,Name=\"dyn_Skirt_01_SB_l\"),Transform=(Rotation=(X=0,Y=0.1736,Z=0,W=0.9848),Translation=(X=0,Y=0,Z=0),Scale3D=(X=1,Y=1,Z=1))),(Item=(Type=Bone,Name=\"dyn_Skirt_01_SB_r\"),Transform=(Rotation=(X=0,Y=0.1736,Z=0,W=0.9848),Translation=(X=0,Y=0,Z= ...(truncated)"
    WeightMinimum=0
    Weight=@AlphaInterp.Result
   )
   n1 = &RigVMFunction_AlphaInterp(
    InterpSpeedDecreasing=8
    InterpSpeedIncreasing=15
    Scale=1
    bClampResult=False
    Value=&RigVMFunction_MathBoolToFloat(Value=RigVMVariableNode() {variable=isDiveMode}.Value).Result
   )
   > &RigUnit_ModifyTransforms(
    ItemToModify="((Item=(Type=Bone,Name=\"dyn_Skirt_05_SB_l\"),Transform=(Rotation=(X=0,Y=-0.2588,Z=0,W=0.9659),Translation=(X=0,Y=0,Z=0),Scale3D=(X=1,Y=1,Z=1))),(Item=(Type=Bone,Name=\"dyn_Skirt_05_SB_r\"),Transform=(Rotation=(X=0,Y=-0.2588,Z=0,W=0.9659),Translation=(X=0,Y=0, ...(truncated)"
    WeightMinimum=0
    Weight=@n1.Result
   )
   Multiply = &RigVMFunction_MathDoubleMul(
    A=&RigVMFunction_MathBoolToFloat(Value=RigVMVariableNode() {variable=isDiveMode}.Value).Result
    B=RigVMVariableNode() {variable=moveLeft}.Value
   )
   n2 = &RigVMFunction_AlphaInterp(InterpSpeedDecreasing=12, InterpSpeedIncreasing=10, Scale=1, bClampResult=False, Value=@Multiply.Result)
   > &RigUnit_ModifyTransforms(
    ItemToModify="((Item=(Type=Bone,Name=\"dyn_Skirt_02_S_r\"),Transform=(Rotation=(X=0.0085,Y=0.1302,Z=-0.0648,W=0.9893),Translation=(X=0,Y=0,Z=0),Scale3D=(X=1,Y=1,Z=1))),(Item=(Type=Bone,Name=\"dyn_Skirt_02_SB_r\"),Transform=(Rotation=(X=0.13,Y=0.0864,Z=-0.0114,W=0.9877),Tran ...(truncated)"
    WeightMinimum=0
    Weight=@n2.Result
   )
   n3 = RigVMVariableNode() {variable=isDiveMode}
   n4 = &RigVMFunction_AlphaInterp(
    InterpSpeedDecreasing=12
    InterpSpeedIncreasing=10
    Scale=1
    bClampResult=False
    Value=&RigVMFunction_MathDoubleMul(A=&RigVMFunction_MathBoolToFloat(Value=@n3.Value).Result, B=RigVMVariableNode() {variable=moveRight}.Value).Result
   )
   > &RigUnit_ModifyTransforms(
    ItemToModify="((Item=(Type=Bone,Name=\"dyn_Skirt_02_S_l\"),Transform=(Rotation=(X=0.004,Y=0.1521,Z=-0.0259,W=0.988),Translation=(X=0,Y=0,Z=0),Scale3D=(X=1,Y=1,Z=1))),(Item=(Type=Bone,Name=\"dyn_Skirt_01_SB_l\"),Transform=(Rotation=(X=0.0436,Y=0.0436,Z=-0.0019,W=0.9981),Tran ...(truncated)"
    WeightMinimum=0
    Weight=@n4.Result
   )
   > RigVMFunctionReturnNode()
   &RigUnit_ModifyTransforms(
    ExecutePin=()
    ItemToModify="((Item=(Type=Bone,Name=\"dyn_Skirt_01_SB_l\"),Transform=(Rotation=(X=0,Y=0.1736,Z=0,W=0.9848),Translation=(X=0,Y=0,Z=0),Scale3D=(X=1,Y=1,Z=1))),(Item=(Type=Bone,Name=\"dyn_Skirt_01_SB_r\"),Transform=(Rotation=(X=0,Y=0.1736,Z=0,W=0.9848),Translation=(X=0,Y=0,Z= ...(truncated)"
    Weight=1
    WeightMinimum=0
   )
   &RigUnit_ModifyTransforms(
    ExecutePin=()
    ItemToModify="((Item=(Type=Bone,Name=\"dyn_Skirt_02_S_r\"),Transform=(Rotation=(X=0.0085,Y=0.1302,Z=-0.0648,W=0.9893),Translation=(X=0,Y=0,Z=0),Scale3D=(X=1,Y=1,Z=1))),(Item=(Type=Bone,Name=\"dyn_Skirt_02_SB_r\"),Transform=(Rotation=(X=0.13,Y=0.0864,Z=-0.0114,W=0.9877),Tran ...(truncated)"
    Weight=1
    WeightMinimum=0
   )
   &RigUnit_ModifyTransforms(
    ExecutePin=()
    ItemToModify="((Item=(Type=Bone,Name=\"dyn_Skirt_02_S_l\"),Transform=(Rotation=(X=0.004,Y=0.1521,Z=-0.0259,W=0.988),Translation=(X=0,Y=0,Z=0),Scale3D=(X=1,Y=1,Z=1))),(Item=(Type=Bone,Name=\"dyn_Skirt_01_SB_l\"),Transform=(Rotation=(X=0.0436,Y=0.0436,Z=-0.0019,W=0.9981),Tran ...(truncated)"
    Weight=1
    WeightMinimum=0
   )
  ]}
 RigVMFunctionLibrary {nodes [
   FE_IdleFixes = RigVMCollapseNode(ExecuteContext=())
   LowFpsFixes = RigVMCollapseNode(ExecuteContext=())
   NonDynamicSkydiveFixes = RigVMCollapseNode(ExecuteContext=())
   SkyFloatingCorrector = RigVMCollapseNode(ExecuteContext=())
   braceletPositionHolder = RigVMCollapseNode(ExecuteContext=())
   skirtKey Corrector = RigVMCollapseNode(ExecuteContext=())
  ]}
 RigVMModel {nodes [
   RigUnit_BeginExecution = &RigUnit_BeginExecution()
   braceletPositionHolder = RigVMFunctionReferenceNode(ExecuteContext=@RigUnit_BeginExecution.ExecutePin)
   RigVMFunction_ControlFlowBranch = &RigVMFunction_ControlFlowBranch(Condition=RigVMVariableNode() {variable=isRbanEnabled}.Value, ExecuteContext=@braceletPositionHolder.ExecuteContext)
   skirtKey Corrector = RigVMFunctionReferenceNode(ExecuteContext=@RigVMFunction_ControlFlowBranch.Completed)
   SkyFloatingCorrector = RigVMFunctionReferenceNode(ExecuteContext=@RigVMFunction_ControlFlowBranch.False)
   > LowFpsFixes = RigVMFunctionReferenceNode()
   > FE_IdleFixes = RigVMFunctionReferenceNode()
   > NonDynamicSkydiveFixes = RigVMFunctionReferenceNode()
  ]}
]
hierarchy [
 root {loc=<0 0 0>, rot=<0 0 0 1>}
 pelvis {parent=root, loc=<0 -0.5301 98.0571>, rot=<0.0127 0.707 -0.0127 -0.707>}
 spine_01 {parent=pelvis, loc=<6.1208 -0.9328 0>, rot=<0 0 0.034 -0.9994>}
 spine_02 {parent=spine_01, loc=<6.5997 -0.2095 0>, rot=<0 0 0.0238 -0.9997>}
 spine_03 {parent=spine_02, loc=<6.4914 -0.1237 0>, rot=<0 0 0.0149 -0.9999>}
 spine_04 {parent=spine_03, loc=<6.9598 0.0794 0>, rot=<0 0 -0.0132 -0.9999>}
 spine_05 {parent=spine_04, loc=<6.9987 0 0>, rot=<0 0 -0.0015 -1>}
 neck_01 {parent=spine_05, loc=<11.3138 0.1529 0>, rot=<0 0 -0.2003 -0.9797>}
 neck_02 {parent=neck_01, loc=<4.5627 0 0>, rot=<0 0 0.0504 -0.9987>}
 head {parent=neck_02, loc=<5.3914 0 0>, rot=<0 0 0.115 -0.9934>}
 clavicle_l {parent=spine_05, loc=<5.4831 -0.1053 -3.2946>, rot=<0.0465 -0.7289 -0.009 -0.6829>}
 upperarm_l {parent=clavicle_l, loc=<9.8658 -1.2136 -0.4105>, rot=<0.0164 -0.3245 -0.0269 -0.9454>}
 lowerarm_l {parent=upperarm_l, loc=<26.465 -0.8345 0.259>, rot=<0.0036 0.0161 -0.2298 -0.9731>}
 hand_l {parent=lowerarm_l, loc=<22.7355 -0.3203 -0.6105>, rot=<-0.6784 -0.0223 0.0735 -0.7307>}
 thumb_01_l {parent=hand_l, loc=<1.3118 -0.3191 -2.459>, rot=<0.6546 -0.4042 0.0664 -0.6354>}
 thumb_02_l {parent=thumb_01_l, loc=<5.9336 -0.0448 0.0401>, rot=<-0.0037 0.0213 0.298 -0.9543>}
 thumb_03_l {parent=thumb_02_l, loc=<2.2223 0.3682 -0.0907>, rot=<0.0118 -0.013 0.0066 -0.9998>}
 index_metacarpal_l {parent=hand_l, loc=<1.6535 0.0183 -2.96>, rot=<0.0158 -0.0124 -0.0173 -0.9996>}
 index_01_l {parent=index_metacarpal_l, loc=<8.3008 0.0357 -0.0524>, rot=<-0.0044 0.0485 0.0924 -0.9945>}
 index_02_l {parent=index_01_l, loc=<3.3587 -0.3004 0.0452>, rot=<-0.002 0.0027 0.3225 -0.9466>}
 index_03_l {parent=index_02_l, loc=<2.5578 -0.123 -0.0252>, rot=<-0.002 0.0067 0.2558 -0.9667>}
 middle_metacarpal_l {parent=hand_l, loc=<2.1856 -0.1725 -0.5984>, rot=<0.006 0.0168 -0.0329 -0.9993>}
 middle_01_l {parent=middle_metacarpal_l, loc=<7.3275 -0.2447 0.0129>, rot=<-0.0208 0.0042 0.0858 -0.9961>}
 middle_02_l {parent=middle_01_l, loc=<4.0123 0.1401 0.1175>, rot=<-0.0218 0.0258 0.3512 -0.9357>}
 middle_03_l {parent=middle_02_l, loc=<2.6077 0.0534 0.1373>, rot=<0.0082 0.0188 0.1708 -0.9851>}
 ring_metacarpal_l {parent=hand_l, loc=<2.2074 -0.3625 1.3638>, rot=<-0.0284 0.0722 -0.0458 -0.9959>}
 ring_01_l {parent=ring_metacarpal_l, loc=<7.0295 -0.6197 -0.1446>, rot=<-0.0489 -0.0252 0.1892 -0.9804>}
 ring_02_l {parent=ring_01_l, loc=<3.6174 0.4647 -0.1027>, rot=<0 0 0.2806 -0.9598>}
 ring_03_l {parent=ring_02_l, loc=<2.4152 0.114 0.0199>, rot=<0 0 0.1189 -0.9929>}
 pinky_metacarpal_l {parent=hand_l, loc=<1.7616 -0.9753 3.3814>, rot=<-0.0397 0.0814 -0.069 -0.9935>}
 pinky_01_l {parent=pinky_metacarpal_l, loc=<7.1709 -0.4831 -0.1094>, rot=<-0.0654 -0.0035 0.2242 -0.9723>}
 pinky_02_l {parent=pinky_01_l, loc=<2.8274 0.0893 -0.1043>, rot=<0.0037 0.0011 0.2854 -0.9584>}
 pinky_03_l {parent=pinky_02_l, loc=<2.2046 0.0307 -0.0078>, rot=<-0.0002 0.0098 0.1445 -0.9895>}
 deform_wrist_dn_base_l {parent=hand_l, loc=<0 -1.5326 0>, rot=<0 0 0 -1>}
 deform_wrist_dn_l {parent=deform_wrist_dn_base_l, loc=<0 0 0>, rot=<0 0 0 1>}
 deform_wrist_up_base_l {parent=hand_l, loc=<0 2.3009 0>, rot=<0 0 0 -1>}
 deform_wrist_up_l {parent=deform_wrist_up_base_l, loc=<0 0 0>, rot=<0 0 0 1>}
 lowerarm_twist_02_l {parent=lowerarm_l, loc=<10.6707 0 0>, rot=<0 0 0 1>}
 lowerarm_twist_01_l {parent=lowerarm_l, loc=<18.316 0 0>, rot=<0 0 0 1>}
 deform_elbow_in_base_l {parent=lowerarm_l, loc=<0.1834 3.6062 -0.2572>, rot=<0 0 0 -1>}
 deform_elbow_in_l {parent=deform_elbow_in_base_l, loc=<0 0 0>, rot=<0 0 0 1>}
 deform_elbow_out_base_l {parent=lowerarm_l, loc=<0 -4 0>, rot=<0 0 0 1>}
 deform_elbow_out_l {parent=deform_elbow_out_base_l, loc=<0 0 0>, rot=<0 0 0 1>}
 upperarm_twist_01_l {parent=upperarm_l, loc=<7.2135 -0.1134 0>, rot=<0 0 0 1>}
 deform_upperarm_bk_base_l {parent=upperarm_twist_01_l, loc=<-6.4053 -4.3746 0.314>, rot=<0 0 0 -1>}
 deform_upperarm_bk_l {parent=deform_upperarm_bk_base_l, loc=<0 0 0>, rot=<0 0 0 1>}
 deform_upperarm_fr_base_l {parent=upperarm_twist_01_l, loc=<-5.6312 5.092 0.3354>, rot=<0 0 0 -1>}
 deform_upperarm_fr_l {parent=deform_upperarm_fr_base_l, loc=<0 0 0>, rot=<0 0 0 1>}
 deform_upperarm_up_base_l {parent=upperarm_twist_01_l, loc=<-8.891 0.4825 3.9517>, rot=<0 0 0 -1>}
 deform_upperarm_up_l {parent=deform_upperarm_up_base_l, loc=<0 0 0>, rot=<0 0 0 1>}
 upperarm_twist_bicepPush_l_base {parent=upperarm_twist_01_l, loc=<0 0 1.3>, rot=<0 0 0 1>}
 upperarm_twist_bicepPush_l {parent=upperarm_twist_bicepPush_l_base, loc=<0 0 0>, rot=<0 0 0 1>}
 upperarm_twist_02_l {parent=upperarm_l, loc=<19.7998 -0.1134 0>, rot=<0 0 0 1>}
 clavicle_l_AdjJnt_base {parent=clavicle_l, loc=<3.5471 0 4.1202>, rot=<0 0 0 1>}
 clavicle_l_AdjJnt {parent=clavicle_l_AdjJnt_base, loc=<0 0 0>, rot=<0 0 0 1>}
 clavicle_r {parent=spine_05, loc=<5.4831 -0.1053 3.2946>, rot=<0.7289 0.0465 -0.6829 0.009>}
 upperarm_r {parent=clavicle_r, loc=<-9.8658 1.2136 0.4105>, rot=<0.0164 -0.3245 -0.0269 -0.9454>}
 lowerarm_r {parent=upperarm_r, loc=<-26.465 0.8345 -0.259>, rot=<0.0036 0.0161 -0.2298 -0.9731>}
 hand_r {parent=lowerarm_r, loc=<-22.7355 0.3203 0.6106>, rot=<-0.6784 -0.0223 0.0735 -0.7307>}
 thumb_01_r {parent=hand_r, loc=<-1.3118 0.3191 2.459>, rot=<0.6546 -0.4042 0.0664 -0.6354>}
 thumb_02_r {parent=thumb_01_r, loc=<-5.9336 0.0448 -0.0401>, rot=<-0.0037 0.0213 0.298 -0.9543>}
 thumb_03_r {parent=thumb_02_r, loc=<-2.2223 -0.3682 0.0907>, rot=<0.0118 -0.013 0.0066 -0.9998>}
 index_metacarpal_r {parent=hand_r, loc=<-1.6535 -0.0183 2.96>, rot=<0.0158 -0.0124 -0.0173 -0.9996>}
 index_01_r {parent=index_metacarpal_r, loc=<-8.3008 -0.0357 0.0524>, rot=<-0.0044 0.0485 0.0924 -0.9945>}
 index_02_r {parent=index_01_r, loc=<-3.3587 0.3004 -0.0452>, rot=<-0.002 0.0027 0.3225 -0.9466>}
 index_03_r {parent=index_02_r, loc=<-2.5578 0.123 0.0252>, rot=<-0.002 0.0067 0.2558 -0.9667>}
 middle_metacarpal_r {parent=hand_r, loc=<-2.1856 0.1725 0.5984>, rot=<0.006 0.0168 -0.0329 -0.9993>}
 middle_01_r {parent=middle_metacarpal_r, loc=<-7.3275 0.2447 -0.0129>, rot=<-0.0208 0.0042 0.0858 -0.9961>}
 middle_02_r {parent=middle_01_r, loc=<-4.0123 -0.1401 -0.1175>, rot=<-0.0218 0.0258 0.3512 -0.9357>}
 middle_03_r {parent=middle_02_r, loc=<-2.6077 -0.0534 -0.1373>, rot=<0.0082 0.0188 0.1708 -0.9851>}
 ring_metacarpal_r {parent=hand_r, loc=<-2.2074 0.3625 -1.3638>, rot=<-0.0284 0.0722 -0.0458 -0.9959>}
 ring_01_r {parent=ring_metacarpal_r, loc=<-7.0295 0.6197 0.1446>, rot=<-0.0489 -0.0252 0.1892 -0.9804>}
 ring_02_r {parent=ring_01_r, loc=<-3.6174 -0.4647 0.1027>, rot=<0 0 0.2806 -0.9598>}
 ring_03_r {parent=ring_02_r, loc=<-2.4152 -0.114 -0.0199>, rot=<0 0 0.1189 -0.9929>}
 pinky_metacarpal_r {parent=hand_r, loc=<-1.7616 0.9753 -3.3814>, rot=<-0.0397 0.0814 -0.069 -0.9935>}
 pinky_01_r {parent=pinky_metacarpal_r, loc=<-7.1709 0.4831 0.1094>, rot=<-0.0654 -0.0035 0.2242 -0.9723>}
 pinky_02_r {parent=pinky_01_r, loc=<-2.8274 -0.0893 0.1043>, rot=<0.0037 0.0011 0.2854 -0.9584>}
 pinky_03_r {parent=pinky_02_r, loc=<-2.2046 -0.0307 0.0078>, rot=<-0.0002 0.0098 0.1445 -0.9895>}
 deform_wrist_dn_base_r {parent=hand_r, loc=<-0.0001 1.5328 0>, rot=<0 0 0 -1>}
 deform_wrist_dn_r {parent=deform_wrist_dn_base_r, loc=<0 0 0>, rot=<0 0 0 1>}
 deform_wrist_up_base_r {parent=hand_r, loc=<-0.0002 -2.3007 -0.0001>, rot=<0 0 0 -1>}
 deform_wrist_up_r {parent=deform_wrist_up_base_r, loc=<0 0 0>, rot=<0 0 0 1>}
 lowerarm_twist_02_r {parent=lowerarm_r, loc=<-10.6707 0 0>, rot=<0 0 0 -1>}
 lowerarm_twist_01_r {parent=lowerarm_r, loc=<-18.316 0 0>, rot=<0 0 0 -1>}
 base_wristChain_01_r {parent=lowerarm_twist_01_r, loc=<-3.7279 -1.3373 2.6445>, rot=<0.5985 0.0419 0.7999 -0.0107>}
 dyn_wristChain_01_r {parent=base_wristChain_01_r, loc=<0 0 0>, rot=<0 0 0 1>}
 dyn_wristChain_02_r {parent=dyn_wristChain_01_r, loc=<1.1108 0 0>, rot=<-0.4493 -0.0111 0.0322 -0.8927>}
 dyn_wristChain_03_r {parent=dyn_wristChain_02_r, loc=<2.8056 0 0>, rot=<0.2861 0 0 -0.9582>}
 deform_elbow_in_base_r {parent=lowerarm_r, loc=<-0.1834 -3.6062 0.2572>, rot=<0 0 0 -1>}
 deform_elbow_in_r {parent=deform_elbow_in_base_r, loc=<0 0 0>, rot=<0 0 0 1>}
 deform_elbow_out_base_r {parent=lowerarm_r, loc=<0 4 -0.0001>, rot=<0 0 0 -1>}
 deform_elbow_out_r {parent=deform_elbow_out_base_r, loc=<0 0 0>, rot=<0 0 0 1>}
 upperarm_twist_01_r {parent=upperarm_r, loc=<-7.2135 0.1134 0>, rot=<0 0 0 -1>}
 deform_upperarm_bk_base_r {parent=upperarm_twist_01_r, loc=<6.4053 4.3746 -0.314>, rot=<0 0 0 -1>}
 deform_upperarm_bk_r {parent=deform_upperarm_bk_base_r, loc=<0 0 0>, rot=<0 0 0 1>}
 deform_upperarm_fr_base_r {parent=upperarm_twist_01_r, loc=<5.6312 -5.092 -0.3354>, rot=<0 0 0 -1>}
 deform_upperarm_fr_r {parent=deform_upperarm_fr_base_r, loc=<0 0 0>, rot=<0 0 0 1>}
 deform_upperarm_up_base_r {parent=upperarm_twist_01_r, loc=<8.891 -0.4825 -3.9517>, rot=<0 0 0 -1>}
 deform_upperarm_up_r {parent=deform_upperarm_up_base_r, loc=<0 0 0>, rot=<0 0 0 1>}
 upperarm_twist_bicepPush_r_base {parent=upperarm_twist_01_r, loc=<0 0 -1.3>, rot=<0 0 0 1>}
 upperarm_twist_bicepPush_r {parent=upperarm_twist_bicepPush_r_base, loc=<0 0 0>, rot=<0 0 0 1>}
 upperarm_twist_02_r {parent=upperarm_r, loc=<-19.7998 0.1134 0>, rot=<0 0 0 -1>}
 clavicle_r_AdjJnt_base {parent=clavicle_r, loc=<-3.5471 0 -4.1203>, rot=<0 0 0 1>}
 clavicle_r_AdjJnt {parent=clavicle_r_AdjJnt_base, loc=<0 0 0>, rot=<0 0 0 1>}
 attach_backpack {parent=spine_05, loc=<-0.013 -9.2707 0>, rot=<-0.0093 -0.707 0.0093 -0.707>}
 Backpack_BR {parent=attach_backpack, type=Null, loc=<0 0 0>, rot=<0 0 0 1>}
 attach_cape {parent=spine_05, loc=<8.2648 -7.448 0>, rot=<0.0616 -0.7044 -0.0616 -0.7044>}
 cape {parent=attach_cape, type=Null, loc=<0 0 0>, rot=<0 0 0 1>}
 deform_pec_aim_base_l {parent=spine_04, loc=<3.2456 9.3309 -1.2237>, rot=<0.0603 -0.4033 0.3318 -0.8507>}
 deform_pec_aim_l {parent=deform_pec_aim_base_l, loc=<0 0 0>, rot=<0 0 0 1>}
 deform_pec_base_l {parent=deform_pec_aim_l, loc=<9.3665 0 0>, rot=<0 0 0 1>}
 deform_pec_l {parent=deform_pec_base_l, loc=<0 0 0>, rot=<0 0 0 1>}
 deform_pec_aim_base_r {parent=spine_04, loc=<3.2452 9.3309 1.2237>, rot=<-0.4033 -0.0603 0.8507 0.3318>}
 deform_pec_aim_r {parent=deform_pec_aim_base_r, loc=<0 0 0>, rot=<0 0 0 1>}
 deform_pec_base_r {parent=deform_pec_aim_r, loc=<-9.3667 -0.0003 -0.0004>, rot=<0 0 0 1>}
 deform_pec_r {parent=deform_pec_base_r, loc=<0 0 0>, rot=<0 0 0 1>}
 base_beltStrap {parent=spine_02, loc=<-1.8769 3.3393 10.0404>, rot=<-0.277 0.6532 0.3031 -0.6361>}
 dyn_beltStrap {parent=base_beltStrap, loc=<0 0 0>, rot=<0 0 0 1>}
 base_key_Start_l {parent=spine_01, loc=<5.2317 -2.3134 -8.9283>, rot=<0.0644 -0.929 0.257 -0.2585>}
 dyn_key_Start_l {parent=base_key_Start_l, loc=<0 0 0>, rot=<0 0 0 1>}
 dyn_key_Mid_l {parent=dyn_key_Start_l, loc=<3.3066 0.015 0.0673>, rot=<-0.0149 0.0687 -0.0126 -0.9974>}
 base_key_l {parent=dyn_key_Mid_l, loc=<1.235 -0.0572 -0.3075>, rot=<0.0669 -0.1077 -0.0373 -0.9912>}
 dyn_key_l {parent=base_key_l, loc=<0 0 0>, rot=<0 0 0 1>}
 base_Skirt_01_F_l {parent=spine_01, loc=<2.2295 4.9463 -8.878>, rot=<0.0256 -0.9501 -0.2501 -0.1848>}
 auto_Skirt_01_F_l {parent=base_Skirt_01_F_l, loc=<0 0 0>, rot=<0 0 0 1>}
 offset_Skirt_01_F_l {parent=auto_Skirt_01_F_l, loc=<0 0 0>, rot=<0 0 0 1>}
 dyn_Skirt_01_F_l {parent=offset_Skirt_01_F_l, loc=<0 0 0>, rot=<0 0 0 1>}
 dyn_Skirt_02_F_l {parent=dyn_Skirt_01_F_l, loc=<9.0869 -0.1265 0.1273>, rot=<-0.0043 -0.0367 -0.0362 -0.9987>}
 dyn_Skirt_03_F_l {parent=dyn_Skirt_02_F_l, loc=<12.2093 -0.0465 0.2812>, rot=<0.0046 -0.0596 -0.0102 -0.9982>}
 dyn_Skirt_04_F_l {parent=dyn_Skirt_03_F_l, loc=<14.1462 0 0>, rot=<0.0022 -0.052 -0.0071 -0.9986>}
 dyn_Skirt_05_F_l {parent=dyn_Skirt_04_F_l, loc=<12.0618 0 0>, rot=<0 0 0 1>}
 base_Skirt_01_S_l {parent=spine_01, loc=<3.0629 -3.1027 -9.2001>, rot=<0.0637 -0.9715 0.067 -0.2182>}
 auto_Skirt_01_S_l {parent=base_Skirt_01_S_l, loc=<0 0 0>, rot=<0 0 0 1>}
 offset_Skirt_01_S_l {parent=auto_Skirt_01_S_l, loc=<0 0 0>, rot=<0 0 0 1>}
 dyn_Skirt_01_S_l {parent=offset_Skirt_01_S_l, loc=<0 0 0>, rot=<0 0 0 1>}
 dyn_Skirt_02_S_l {parent=dyn_Skirt_01_S_l, loc=<13.5698 0 0>, rot=<-0.0207 -0.1179 -0.0429 -0.9919>}
 dyn_Skirt_03_S_l {parent=dyn_Skirt_02_S_l, loc=<10.0701 0 0>, rot=<0.001 -0.0475 0.0147 -0.9988>}
 dyn_Skirt_04_S_l {parent=dyn_Skirt_03_S_l, loc=<11.7867 0 0>, rot=<-0.0002 -0.0168 0.0013 -0.9999>}
 dyn_Skirt_05_S_l {parent=dyn_Skirt_04_S_l, loc=<12.0618 0 0>, rot=<0 0 0 1>}
 base_Skirt_01_SB_l {parent=spine_01, loc=<3.5762 -6.3523 -5.5967>, rot=<0.1307 -0.7474 0.6351 -0.1447>}
 auto_Skirt_01_SB_l {parent=base_Skirt_01_SB_l, loc=<0 0 0>, rot=<0 0 0 1>}
 offset_Skirt_01_SB_l {parent=auto_Skirt_01_SB_l, loc=<0 0 0>, rot=<0 0 0 1>}
 dyn_Skirt_01_SB_l {parent=offset_Skirt_01_SB_l, loc=<0 0 0>, rot=<0 0 0 1>}
 dyn_Skirt_02_SB_l {parent=dyn_Skirt_01_SB_l, loc=<13.5698 0 0>, rot=<-0.0207 -0.1179 -0.0429 -0.9919>}
 dyn_Skirt_03_SB_l {parent=dyn_Skirt_02_SB_l, loc=<10.0701 0 0>, rot=<0.001 -0.0475 0.0147 -0.9988>}
 dyn_Skirt_04_SB_l {parent=dyn_Skirt_03_SB_l, loc=<11.7867 0 0>, rot=<-0.0002 -0.0168 0.0013 -0.9999>}
 dyn_Skirt_05_SB_l {parent=dyn_Skirt_04_SB_l, loc=<12.0618 0 0>, rot=<0 0 0.0293 -0.9996>}
 dyn_SkirtStrap_06_SB_l {parent=dyn_Skirt_05_SB_l, loc=<6.8745 0 0>, rot=<0 0 0 1>}
 dyn_SkirtStrap_07_SB_l {parent=dyn_SkirtStrap_06_SB_l, loc=<7.0255 0 0>, rot=<0 0 0 1>}
 base_Skirt_01_SB_r {parent=spine_01, loc=<3.7885 -5.7639 5.4646>, rot=<0.1794 -0.6896 0.6952 -0.0949>}
 auto_Skirt_01_SB_r {parent=base_Skirt_01_SB_r, loc=<0 0 0>, rot=<0 0 0 1>}
 offset_Skirt_01_SB_r {parent=auto_Skirt_01_SB_r, loc=<0 0 0>, rot=<0 0 0 1>}
 dyn_Skirt_01_SB_r {parent=offset_Skirt_01_SB_r, loc=<0 0 0>, rot=<0 0 0 1>}
 dyn_Skirt_02_SB_r {parent=dyn_Skirt_01_SB_r, loc=<13.5698 0 0>, rot=<-0.0207 -0.1179 -0.0429 -0.9919>}
 dyn_Skirt_03_SB_r {parent=dyn_Skirt_02_SB_r, loc=<10.0701 0 0>, rot=<0.001 -0.0475 0.0147 -0.9988>}
 dyn_Skirt_04_SB_r {parent=dyn_Skirt_03_SB_r, loc=<11.7867 0 0>, rot=<-0.0002 -0.0168 0.0013 -0.9999>}
 dyn_Skirt_05_SB_r {parent=dyn_Skirt_04_SB_r, loc=<12.0618 0 0>, rot=<0 0 -0.0279 -0.9996>}
 dyn_SkirtStrap_06_SB_r {parent=dyn_Skirt_05_SB_r, loc=<6.8745 0 0>, rot=<0 0 0 1>}
 dyn_SkirtStrap_07_SB_r {parent=dyn_SkirtStrap_06_SB_r, loc=<7.0255 0 0>, rot=<0 0 0 1>}
 base_Skirt_01_S_r {parent=spine_01, loc=<3.0633 -3.1027 9.2001>, rot=<0.226 -0.2347 0.9451 -0.0248>}
 auto_Skirt_01_S_r {parent=base_Skirt_01_S_r, loc=<0 0 0>, rot=<0 0 0 1>}
 offset_Skirt_01_S_r {parent=auto_Skirt_01_S_r, loc=<0 0 0>, rot=<0 0 0 1>}
 dyn_Skirt_01_S_r {parent=offset_Skirt_01_S_r, loc=<0 0 0>, rot=<0 0 0 1>}
 dyn_Skirt_02_S_r {parent=dyn_Skirt_01_S_r, loc=<13.5702 0 0>, rot=<0.0207 -0.1255 -0.0001 -0.9919>}
 dyn_Skirt_03_S_r {parent=dyn_Skirt_02_S_r, loc=<10.07 0 0>, rot=<-0.001 -0.0396 -0.0301 -0.9988>}
 dyn_Skirt_04_S_r {parent=dyn_Skirt_03_S_r, loc=<11.7867 0 0>, rot=<0.0002 -0.0153 -0.007 -0.9999>}
 dyn_Skirt_05_S_r {parent=dyn_Skirt_04_S_r, loc=<12.0618 0 0>, rot=<0 0 0 1>}
 base_Skirt_01_F_r {parent=spine_01, loc=<2.2293 4.9463 8.878>, rot=<0.1931 0.2499 0.9483 -0.0307>}
 auto_Skirt_01_F_r {parent=base_Skirt_01_F_r, loc=<0 0 0>, rot=<0 0 0 1>}
 offset_Skirt_01_F_r {parent=auto_Skirt_01_F_r, loc=<0 0 0>, rot=<0 0 0 1>}
 dyn_Skirt_01_F_r {parent=offset_Skirt_01_F_r, loc=<0 0 0>, rot=<0 0 0 1>}
 dyn_Skirt_02_F_r {parent=dyn_Skirt_01_F_r, loc=<9.0885 0 0>, rot=<0.007 -0.0323 0.0411 -0.9986>}
 dyn_Skirt_03_F_r {parent=dyn_Skirt_02_F_r, loc=<12.2126 0 0>, rot=<-0.0058 -0.0711 0.012 -0.9974>}
 dyn_Skirt_04_F_r {parent=dyn_Skirt_03_F_r, loc=<14.1462 0 0>, rot=<-0.0022 -0.0521 0.0071 -0.9986>}
 dyn_Skirt_05_F_r {parent=dyn_Skirt_04_F_r, loc=<12.0618 0 0>, rot=<0 0 0 1>}
 dyn_Skirt_01_SB_l_Spine01DeltaRef_null {parent=spine_01, type=Null, loc=<3.5762 -6.3523 -5.5967>, rot=<-0.1307 0.7474 -0.6351 0.1447>}
 thigh_l {parent=pelvis, loc=<-1.0855 -0.2109 -8.7053>, rot=<0.0009 0.0355 0.0155 -0.9993>}
 calf_l {parent=thigh_l, loc=<-42.5284 -0.0019 -0.7297>, rot=<0.0006 -0.0161 -0.0391 -0.9991>}
 foot_l {parent=calf_l, loc=<-46.7401 -2.5014 -0.592>, rot=<0.0427 -0.0238 0.0353 -0.9982>}
 ball_l {parent=foot_l, loc=<-3.5658 10.8066 -0.1544>, rot=<0.0001 -0.0001 -0.7063 -0.7079>}
 calf_twist_02_l {parent=calf_l, loc=<-19.1546 -1.7943 -0.2638>, rot=<0.0237 0.0095 -0.0035 -0.9997>}
 base_bootStrap_l {parent=calf_twist_02_l, loc=<-5.4535 -6.25 -0.837>, rot=<-0.0445 -0.9982 -0.0257 0.0299>}
 dyn_bootStrap_l {parent=base_bootStrap_l, loc=<0 0 0>, rot=<0 0 0 1>}
 calf_twist_01_l {parent=calf_l, loc=<-35.8483 -1.9197 -0.5794>, rot=<0.0237 0.0095 -0.0035 -0.9997>}
 deform_knee_in_base_l {parent=calf_l, loc=<0.8031 -5.0231 0.2456>, rot=<0 0 0 -1>}
 deform_knee_in_l {parent=deform_knee_in_base_l, loc=<0 0 0>, rot=<0 0 0 1>}
 deform_knee_out_base_l {parent=calf_l, loc=<0 4 0>, rot=<0 0 0 -1>}
 deform_knee_out_l {parent=deform_knee_out_base_l, loc=<0 0 0>, rot=<0 0 0 1>}
 thigh_twist_01_l {parent=thigh_l, loc=<-27.952 0 0>, rot=<0.0013 0 0 -1>}
 deform_groin_base_l {parent=thigh_l, loc=<0.4801 5.6776 -0.2486>, rot=<0 0 0 -1>}
 deform_groin_l {parent=deform_groin_base_l, loc=<0 0 0>, rot=<0 0 0 1>}
 thigh_r {parent=pelvis, loc=<-1.0855 -0.2109 8.7053>, rot=<0.0355 -0.0009 0.9993 0.0155>}
 calf_r {parent=thigh_r, loc=<42.5284 0.0019 0.7297>, rot=<0.0006 -0.0161 -0.0391 -0.9991>}
 foot_r {parent=calf_r, loc=<46.7401 2.5014 0.592>, rot=<0.0427 -0.0238 0.0353 -0.9982>}
 ball_r {parent=foot_r, loc=<3.5658 -10.8066 0.1544>, rot=<0.0001 -0.0001 -0.7063 -0.7079>}
 calf_twist_02_r {parent=calf_r, loc=<19.1546 1.7943 0.2638>, rot=<0.0237 0.0095 -0.0035 -0.9997>}
 base_bootStrap_r {parent=calf_twist_02_r, loc=<5.4535 6.2499 0.837>, rot=<0.9982 -0.0445 -0.0299 -0.0257>}
 dyn_bootStrap_r {parent=base_bootStrap_r, loc=<0 0 0>, rot=<0 0 0 1>}
 calf_twist_01_r {parent=calf_r, loc=<35.8483 1.9197 0.5794>, rot=<0.0237 0.0095 -0.0035 -0.9997>}
 deform_knee_in_base_r {parent=calf_r, loc=<-0.8031 5.0231 -0.2457>, rot=<0 0 0 -1>}
 deform_knee_in_r {parent=deform_knee_in_base_r, loc=<0 0 0>, rot=<0 0 0 1>}
 deform_knee_out_base_r {parent=calf_r, loc=<0 -4 0>, rot=<0 0 0 -1>}
 deform_knee_out_r {parent=deform_knee_out_base_r, loc=<0 0 0>, rot=<0 0 0 1>}
 thigh_twist_01_r {parent=thigh_r, loc=<27.952 0 0>, rot=<0.0013 0 0 -1>}
 deform_groin_base_r {parent=thigh_r, loc=<-0.4801 -5.6776 0.2486>, rot=<0 0 0 -1>}
 deform_groin_r {parent=deform_groin_base_r, loc=<0 0 0>, rot=<0 0 0 1>}
 deform_glute_base_l {parent=pelvis, loc=<-7.3466 -9.0213 -9.1245>, rot=<0.0172 0.031 -0.4477 -0.8935>}
 deform_glute_l {parent=deform_glute_base_l, loc=<0 0 0>, rot=<0 0 0 1>}
 deform_glute_base_r {parent=pelvis, loc=<-7.3464 -9.0213 9.1245>, rot=<0.031 -0.0172 0.8935 -0.4477>}
 deform_glute_r {parent=deform_glute_base_r, loc=<0 0 0>, rot=<0 0 0 1>}
 spine_04_PevlisDeltaRef_null {parent=pelvis, type=Null, loc=<26.0226 -3.392 0>, rot=<0 0 -0.0595 0.9982>}
 spine_05_PevlisDeltaRef_null {parent=pelvis, type=Null, loc=<32.9717 -4.2232 0>, rot=<0 0 0.058 -0.9983>}
]
variables [
 CrouchWalking {type=float, default=0}
 MovingBack {type=double, default=0}
 isBaseballSliding {type=double, default=0}
 isCharacterIdle {type=double, default=0}
 isDiveMode {type=bool, default=False}
 isIdleInFE {type=double, default=0}
 isKneeSliding {type=double, default=0}
 isRbanEnabled {type=bool, default=False}
 isSkyFloating {type=bool, default=False}
 lowFPs {type=double, default=0}
 moveLeft {type=double, default=0}
 moveRight {type=double, default=0}
]
