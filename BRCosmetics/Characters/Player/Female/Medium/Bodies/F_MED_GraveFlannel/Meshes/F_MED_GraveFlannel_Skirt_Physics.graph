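# Physics asset for the skirt of the GraveFlannel female medium player cosmetic body.
# Capsule bodies on pelvis, lower spine, legs and feet accompany six five-link skirt chains (front, side and side-back per side), with two strap links ending each side-back chain.
# Chains hang from tiny offset bodies freely jointed to spine_01.
# Skirt links lock twist, keep one swing axis narrow, and widen the other progressively toward the hem.
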PhysicsAsset /BRCosmetics/Characters/Player/Female/Medium/Bodies/F_MED_GraveFlannel/Meshes/F_MED_GraveFlannel_Skirt_Physics
bodies {
 ball_l [{
   shapeType=Capsule
   length=6.4454
   radius=5.8471
   shapeName=ball_l_capsule
   center=<4.9216 0.3298 -0.318>
   rotation=<rot 0.2559 -89.4746 4.5991>
  }]
 ball_r [{
   shapeType=Capsule
   length=6.4454
   radius=5.8471
   shapeName=ball_l_capsule
   center=<-4.4869 -0.1981 0.4944>
   rotation=<rot 0.2559 -89.4746 4.5991>
  }]
 calf_l [{shapeType=Capsule, length=42.1584, radius=7.4555, center=<-22.5716 -3.037 0.2541>, rotation=<rot 0 -89.0161 -91.0358>}]
 calf_r [{shapeType=Capsule, length=41.9261, radius=7.6879, center=<22.5716 3.037 -0.2541>, rotation=<rot 0 -89.0161 88.9642>}]
 dyn_SkirtStrap_06_SB_l [{
   shapeType=Capsule
   length=2.8898
   radius=2.0741
   shapeName=dyn_SkirtStrap_06_SB_l_capsule
   center=<2.6928 -0.0449 -0.107>
   rotation=<rot 0 -90 -90>
  }]
 dyn_SkirtStrap_06_SB_r [{
   shapeType=Capsule
   length=2.8898
   radius=2.0741
   shapeName=dyn_SkirtStrap_06_SB_l_capsule
   center=<2.6928 -0.023 -0.107>
   rotation=<rot 0 -90 -90>
  }]
 dyn_SkirtStrap_07_SB_l [{
   shapeType=Capsule
   length=3.0896
   radius=1.8742
   shapeName=dyn_SkirtStrap_06_SB_l_capsule
   center=<2.6928 -0.0449 -0.107>
   rotation=<rot 0 -90 -90>
  }]
 dyn_SkirtStrap_07_SB_r [{
   shapeType=Capsule
   length=3.0576
   radius=1.9063
   shapeName=dyn_SkirtStrap_06_SB_l_capsule
   center=<2.6928 -0.023 -0.107>
   rotation=<rot 0 -90 -90>
  }]
 dyn_Skirt_01_F_l [{
   shapeType=Capsule
   length=6.8358
   radius=3.5623
   shapeName=dyn_Skirt_01_F_l_capsule
   center=<5.8141 0.1941 2.1541>
   rotation=<rot 0 -92.2646 -89.934>
  }]
 dyn_Skirt_01_F_r [{
   shapeType=Capsule
   length=6.8358
   radius=3.5623
   shapeName=dyn_Skirt_01_F_l_capsule
   center=<5.8181 0.1939 1.5549>
   rotation=<rot 0.015 -92.2976 -91.8892>
  }]
 dyn_Skirt_01_SB_l [{
   shapeType=Capsule
   length=7.1352
   radius=3.2629
   shapeName=dyn_Skirt_01_F_l_capsule
   center=<7.2269 0.1382 2.5279>
   rotation=<rot 0 -92.2646 -95.7367>
  }]
 dyn_Skirt_01_SB_r [{
   shapeType=Capsule
   length=7.6984
   radius=2.9008
   shapeName=dyn_Skirt_01_F_l_capsule
   center=<7.5008 0.1274 2.2053>
   rotation=<rot 0 -92.2646 -94.7005>
  }]
 dyn_Skirt_01_S_l [{
   shapeType=Capsule
   length=6.8358
   radius=3.5623
   shapeName=dyn_Skirt_01_F_l_capsule
   center=<6.399 0.171 3.1342>
   rotation=<rot -0.0004 -92.2638 -95.6957>
  }]
 dyn_Skirt_01_S_r [{
   shapeType=Capsule
   length=6.8358
   radius=3.5623
   shapeName=dyn_Skirt_01_F_l_capsule
   center=<5.7905 0.1951 2.6198>
   rotation=<rot 0.0001 -92.2648 -95.3404>
  }]
 dyn_Skirt_02_F_l [{
   shapeType=Capsule
   length=8.4207
   radius=4.3325
   shapeName=dyn_Skirt_02_F_l_capsule
   center=<5.1017 -0.2904 3.1072>
   rotation=<rot 0.0893 -90.0676 -93.8024>
  }]
 dyn_Skirt_02_F_r [{
   shapeType=Capsule
   length=8.4207
   radius=4.3325
   shapeName=dyn_Skirt_02_F_l_capsule
   center=<5.6304 -0.2907 3.2784>
   rotation=<rot 0.0894 -90.068 -93.3024>
  }]
 dyn_Skirt_02_SB_l [{
   shapeType=Capsule
   length=9.7638
   radius=3.4612
   shapeName=dyn_Skirt_02_F_l_capsule
   center=<4.5246 -0.2908 2.4112>
   rotation=<rot 0.0893 -90.0677 -87.9445>
  }]
 dyn_Skirt_02_SB_r [{
   shapeType=Capsule
   length=9.9108
   radius=3.3142
   shapeName=dyn_Skirt_02_F_l_capsule
   center=<4.5281 -0.2911 2.2212>
   rotation=<rot 0.0214 -89.4086 -84.1182>
  }]
 dyn_Skirt_02_S_l [{
   shapeType=Capsule
   length=7.819
   radius=4.2073
   shapeName=dyn_Skirt_02_F_l_capsule
   center=<4.5481 -0.2904 2.6822>
   rotation=<rot 0.0893 -90.0675 -87.1579>
  }]
 dyn_Skirt_02_S_r [{
   shapeType=Capsule
   length=8.8925
   radius=4.3325
   shapeName=dyn_Skirt_02_F_l_capsule
   center=<4.5601 -0.2912 2.1394>
   rotation=<rot 0.0893 -90.0678 -86.0939>
  }]
 dyn_Skirt_03_F_l [{
   shapeType=Capsule
   length=9.5543
   radius=4.2618
   shapeName=dyn_Skirt_02_F_l_capsule
   center=<6.7877 -0.2935 2.385>
   rotation=<rot 0.0893 -90.0676 -85.7623>
  }]
 dyn_Skirt_03_F_r [{
   shapeType=Capsule
   length=9.8214
   radius=4.2454
   shapeName=dyn_Skirt_02_F_l_capsule
   center=<6.8935 -0.2932 2.6616>
   rotation=<rot 0.0894 -90.0681 -83.7773>
  }]
 dyn_Skirt_03_SB_l [{
   shapeType=Capsule
   length=9.5543
   radius=4.2618
   shapeName=dyn_Skirt_02_F_l_capsule
   center=<7.0948 -0.2956 1.2672>
   rotation=<rot 0.1008 -90.0675 -89.777>
  }]
 dyn_Skirt_03_SB_r [{
   shapeType=Capsule
   length=9.5543
   radius=4.2618
   shapeName=dyn_Skirt_02_F_l_capsule
   center=<7.101 -0.2953 1.4394>
   rotation=<rot 0.0893 -90.0677 -87.9445>
  }]
 dyn_Skirt_03_S_l [{
   shapeType=Capsule
   length=9.4262
   radius=4.3899
   shapeName=dyn_Skirt_02_F_l_capsule
   center=<7.1121 -0.2942 2.1906>
   rotation=<rot 0.0893 -90.0676 -89.5041>
  }]
 dyn_Skirt_03_S_r [{
   shapeType=Capsule
   length=9.5543
   radius=4.2618
   shapeName=dyn_Skirt_02_F_l_capsule
   center=<7.0992 -0.2954 1.3905>
   rotation=<rot 0.0893 -90.0678 -89.9669>
  }]
 dyn_Skirt_04_F_l [{
   shapeType=Capsule
   length=8.746
   radius=5.098
   shapeName=dyn_Skirt_02_F_l_capsule
   center=<7.0203 -3.0026 2.1221>
   rotation=<rot 0.0002 0.0003 -111.0014>
  }]
 dyn_Skirt_04_F_r [{
   shapeType=Capsule
   length=8.746
   radius=5.098
   shapeName=dyn_Skirt_02_F_l_capsule
   center=<7.0203 0.3792 1.8829>
   rotation=<rot 0 0 -96.1345>
  }]
 dyn_Skirt_04_SB_l [{
   shapeType=Capsule
   length=8.746
   radius=5.098
   shapeName=dyn_Skirt_02_F_l_capsule
   center=<4.4724 -0.6485 1.7767>
   rotation=<rot 0 0.0001 -91.2067>
  }]
 dyn_Skirt_04_SB_r [{
   shapeType=Capsule
   length=8.746
   radius=5.098
   shapeName=dyn_Skirt_02_F_l_capsule
   center=<4.4724 -0.1406 1.8781>
   rotation=<rot 0.1354 1.0764 -97.1777>
  }]
 dyn_Skirt_04_S_l [{
   shapeType=Capsule
   length=11.4991
   radius=5.098
   shapeName=dyn_Skirt_02_F_l_capsule
   center=<7.0203 -0.902 2.7334>
   rotation=<rot 0 0.0001 -104.7138>
  }]
 dyn_Skirt_04_S_r [{
   shapeType=Capsule
   length=10.7215
   radius=4.6525
   shapeName=dyn_Skirt_02_F_l_capsule
   center=<7.0203 0.2486 2.1869>
   rotation=<rot 0 0.0001 -97.5108>
  }]
 dyn_Skirt_05_F_l [{
   shapeType=Capsule
   length=5.5559
   radius=5.1467
   shapeName=dyn_Skirt_02_F_l_capsule
   center=<8.3864 -3.1189 1.9725>
   rotation=<rot -1.6849 -15.9909 -111.6125>
  }]
 dyn_Skirt_05_F_r [{
   shapeType=Capsule
   length=2.9208
   radius=5.3924
   shapeName=dyn_Skirt_02_F_l_capsule
   center=<9.1277 1.6935 2.2112>
   rotation=<rot 2.7113 12.9066 -101.6634>
  }]
 dyn_Skirt_05_SB_l [{
   shapeType=Capsule
   length=8.1337
   radius=5.7103
   shapeName=dyn_Skirt_02_F_l_capsule
   center=<2.2413 -0.4294 2.3019>
   rotation=<rot 0.3883 1.832 -101.9639>
  }]
 dyn_Skirt_05_SB_r [{
   shapeType=Capsule
   length=8.7518
   radius=5.0923
   shapeName=dyn_Skirt_02_F_l_capsule
   center=<2.2404 -0.3823 2.3291>
   rotation=<rot -0.2238 -1.0554 -101.968>
  }]
 dyn_Skirt_05_S_l [{
   shapeType=Capsule
   length=10.7332
   radius=4.6066
   shapeName=dyn_Skirt_02_F_l_capsule
   center=<6.1498 -0.4538 2.7376>
   rotation=<rot -3.4492 -14.5832 -103.0205>
  }]
 dyn_Skirt_05_S_r [{
   shapeType=Capsule
   length=9.7628
   radius=4.4534
   shapeName=dyn_Skirt_02_F_l_capsule
   center=<6.0644 -0.2177 3.1467>
   rotation=<rot 2.9368 16.047 -100.1001>
  }]
 foot_l [{shapeType=Capsule, length=0, radius=10.8462, center=<-1.5715 3.8006 -0.2064>, rotation=<rot 0 0 -180>}]
 foot_r [{shapeType=Capsule, length=0.1495, radius=10.8462, center=<1.5715 -3.8006 0.2064>}]
 offset_Skirt_01_F_l [{shapeType=Capsule, length=0.505, radius=0.505, shapeName=offset_Skirt_01_F_l_capsule, rotation=<rot 0 -90 0>}]
 offset_Skirt_01_F_r [{shapeType=Capsule, length=0.505, radius=0.505, shapeName=offset_Skirt_01_F_r_capsule, rotation=<rot 0 -90 0>}]
 offset_Skirt_01_SB_l [{shapeType=Capsule, length=0.505, radius=0.505, shapeName=offset_Skirt_01_SB_l_capsule, rotation=<rot 0 -90 0>}]
 offset_Skirt_01_SB_r [{shapeType=Capsule, length=0.505, radius=0.505, shapeName=offset_Skirt_01_SB_r_capsule, rotation=<rot 0 -90 0>}]
 offset_Skirt_01_S_l [{shapeType=Capsule, length=0.505, radius=0.505, shapeName=offset_Skirt_01_S_l_capsule, rotation=<rot 0 -90 0>}]
 offset_Skirt_01_S_r [{shapeType=Capsule, length=0.505, radius=0.505, shapeName=offset_Skirt_01_S_r_capsule, rotation=<rot 0 -90 0>}]
 pelvis [{shapeType=Capsule, length=10.0202, radius=8.692, shapeName=pelvis_capsule, center=<-2.744 -2.2028 -0.3819>, rotation=<rot 0 -90 0>}]
 spine_01 [{shapeType=Capsule, length=2.6651, radius=8.3268, shapeName=spine_01_capsule, center=<1.6933 0.5052 0>, rotation=<rot 0 -83.7304 90>}]
 thigh_l [
  {shapeType=Capsule, length=37.4632, radius=8.1251, center=<-26.888 -0.2767 1.1876>, rotation=<rot 0.0025 -86.4451 -90.0555>}
  {shapeType=Capsule, length=1.0105, radius=7.3424, shapeName=thigh_l_capsule, center=<-6.3662 -3.6649 1.522>}
 ]
 thigh_r [
  {shapeType=Capsule, length=37.7616, radius=8.2005, center=<27.6194 -0.6722 -1.3968>, rotation=<rot 0.0035 -85.4469 89.9445>}
  {shapeType=Capsule, length=1.0105, radius=7.3424, shapeName=thigh_r_capsule, center=<5.9898 3.069 -1.43>}
 ]
}
constraints [
 {bone1Name=calf_l, bone2Name=thigh_l, swing1LimitDegrees=45, swing2LimitDegrees=45, twistLimitDegrees=45}
 {bone1Name=foot_l, bone2Name=calf_l, swing1LimitDegrees=45, swing2LimitDegrees=45, twistLimitDegrees=45}
 {bone1Name=calf_r, bone2Name=thigh_r, swing1LimitDegrees=45, swing2LimitDegrees=45, twistLimitDegrees=45}
 {bone1Name=foot_r, bone2Name=calf_r, swing1LimitDegrees=45, swing2LimitDegrees=45, twistLimitDegrees=45}
 {
  bone1Name=thigh_l
  bone2Name=pelvis
  swing1LimitDegrees=45
  swing1Motion=Free
  swing2LimitDegrees=45
  swing2Motion=Free
  twistLimitDegrees=45
  twistMotion=Free
 }
 {
  bone1Name=thigh_r
  bone2Name=pelvis
  swing1LimitDegrees=45
  swing1Motion=Free
  swing2LimitDegrees=45
  swing2Motion=Free
  twistLimitDegrees=45
  twistMotion=Free
 }
 {
  bone1Name=spine_01
  bone2Name=pelvis
  swing1LimitDegrees=45
  swing1Motion=Free
  swing2LimitDegrees=45
  swing2Motion=Free
  twistLimitDegrees=45
  twistMotion=Free
 }
 {
  bone1Name=offset_Skirt_01_F_l
  bone2Name=spine_01
  swing1LimitDegrees=45
  swing1Motion=Free
  swing2LimitDegrees=45
  swing2Motion=Free
  twistLimitDegrees=45
  twistMotion=Free
 }
 {bone1Name=dyn_Skirt_01_F_l, bone2Name=offset_Skirt_01_F_l, swing1LimitDegrees=2.5, swing2LimitDegrees=15, twistLimitDegrees=0}
 {bone1Name=dyn_Skirt_02_F_l, bone2Name=dyn_Skirt_01_F_l, swing1LimitDegrees=2.5, swing2LimitDegrees=20, twistLimitDegrees=0}
 {bone1Name=dyn_Skirt_03_F_l, bone2Name=dyn_Skirt_02_F_l, swing1LimitDegrees=2.5, swing2LimitDegrees=25, twistLimitDegrees=0}
 {bone1Name=dyn_Skirt_04_F_l, bone2Name=dyn_Skirt_03_F_l, swing1LimitDegrees=2.5, swing2LimitDegrees=30, twistLimitDegrees=0}
 {bone1Name=dyn_Skirt_05_F_l, bone2Name=dyn_Skirt_04_F_l, swing1LimitDegrees=2.5, swing2LimitDegrees=35, twistLimitDegrees=0}
 {
  bone1Name=offset_Skirt_01_F_r
  bone2Name=spine_01
  swing1LimitDegrees=45
  swing1Motion=Free
  swing2LimitDegrees=45
  swing2Motion=Free
  twistLimitDegrees=45
  twistMotion=Free
 }
 {bone1Name=dyn_Skirt_01_F_r, bone2Name=offset_Skirt_01_F_r, swing1LimitDegrees=2.5, swing2LimitDegrees=15, twistLimitDegrees=0}
 {bone1Name=dyn_Skirt_02_F_r, bone2Name=dyn_Skirt_01_F_r, swing1LimitDegrees=2.5, swing2LimitDegrees=20, twistLimitDegrees=0}
 {bone1Name=dyn_Skirt_03_F_r, bone2Name=dyn_Skirt_02_F_r, swing1LimitDegrees=2.5, swing2LimitDegrees=25, twistLimitDegrees=0}
 {bone1Name=dyn_Skirt_04_F_r, bone2Name=dyn_Skirt_03_F_r, swing1LimitDegrees=2.5, swing2LimitDegrees=30, twistLimitDegrees=0}
 {bone1Name=dyn_Skirt_05_F_r, bone2Name=dyn_Skirt_04_F_r, swing1LimitDegrees=2.5, swing2LimitDegrees=35, twistLimitDegrees=0}
 {
  bone1Name=offset_Skirt_01_SB_l
  bone2Name=spine_01
  swing1LimitDegrees=45
  swing1Motion=Free
  swing2LimitDegrees=45
  swing2Motion=Free
  twistLimitDegrees=45
  twistMotion=Free
 }
 {bone1Name=dyn_Skirt_01_SB_l, bone2Name=offset_Skirt_01_SB_l, swing1LimitDegrees=2.5, swing2LimitDegrees=15, twistLimitDegrees=0}
 {bone1Name=dyn_Skirt_02_SB_l, bone2Name=dyn_Skirt_01_SB_l, swing1LimitDegrees=2.5, swing2LimitDegrees=20, twistLimitDegrees=0}
 {bone1Name=dyn_Skirt_03_SB_l, bone2Name=dyn_Skirt_02_SB_l, swing1LimitDegrees=2.5, swing2LimitDegrees=25, twistLimitDegrees=0}
 {bone1Name=dyn_Skirt_04_SB_l, bone2Name=dyn_Skirt_03_SB_l, swing1LimitDegrees=2.5, swing2LimitDegrees=30, twistLimitDegrees=0}
 {bone1Name=dyn_Skirt_05_SB_l, bone2Name=dyn_Skirt_04_SB_l, swing1LimitDegrees=2.5, swing2LimitDegrees=35, twistLimitDegrees=0}
 {bone1Name=dyn_SkirtStrap_06_SB_l, bone2Name=dyn_Skirt_05_SB_l, swing1LimitDegrees=15, swing2LimitDegrees=45, twistLimitDegrees=0}
 {bone1Name=dyn_SkirtStrap_07_SB_l, bone2Name=dyn_SkirtStrap_06_SB_l, swing1LimitDegrees=15, swing2LimitDegrees=45, twistLimitDegrees=0}
 {
  bone1Name=offset_Skirt_01_SB_r
  bone2Name=spine_01
  swing1LimitDegrees=45
  swing1Motion=Free
  swing2LimitDegrees=45
  swing2Motion=Free
  twistLimitDegrees=45
  twistMotion=Free
 }
 {bone1Name=dyn_Skirt_01_SB_r, bone2Name=offset_Skirt_01_SB_r, swing1LimitDegrees=2.5, swing2LimitDegrees=15, twistLimitDegrees=0}
 {bone1Name=dyn_Skirt_02_SB_r, bone2Name=dyn_Skirt_01_SB_r, swing1LimitDegrees=2.5, swing2LimitDegrees=20, twistLimitDegrees=0}
 {bone1Name=dyn_Skirt_03_SB_r, bone2Name=dyn_Skirt_02_SB_r, swing1LimitDegrees=2.5, swing2LimitDegrees=25, twistLimitDegrees=0}
 {bone1Name=dyn_Skirt_04_SB_r, bone2Name=dyn_Skirt_03_SB_r, swing1LimitDegrees=2.5, swing2LimitDegrees=30, twistLimitDegrees=0}
 {bone1Name=dyn_Skirt_05_SB_r, bone2Name=dyn_Skirt_04_SB_r, swing1LimitDegrees=2.5, swing2LimitDegrees=35, twistLimitDegrees=0}
 {bone1Name=dyn_SkirtStrap_06_SB_r, bone2Name=dyn_Skirt_05_SB_r, swing1LimitDegrees=15, swing2LimitDegrees=45, twistLimitDegrees=0}
 {bone1Name=dyn_SkirtStrap_07_SB_r, bone2Name=dyn_SkirtStrap_06_SB_r, swing1LimitDegrees=15, swing2LimitDegrees=45, twistLimitDegrees=0}
 {
  bone1Name=offset_Skirt_01_S_l
  bone2Name=spine_01
  swing1LimitDegrees=45
  swing1Motion=Free
  swing2LimitDegrees=45
  swing2Motion=Free
  twistLimitDegrees=45
  twistMotion=Free
 }
 {bone1Name=dyn_Skirt_01_S_l, bone2Name=offset_Skirt_01_S_l, swing1LimitDegrees=2.5, swing2LimitDegrees=15, twistLimitDegrees=0}
 {bone1Name=dyn_Skirt_02_S_l, bone2Name=dyn_Skirt_01_S_l, swing1LimitDegrees=2.5, swing2LimitDegrees=20, twistLimitDegrees=0}
 {bone1Name=dyn_Skirt_03_S_l, bone2Name=dyn_Skirt_02_S_l, swing1LimitDegrees=2.5, swing2LimitDegrees=25, twistLimitDegrees=0}
 {bone1Name=dyn_Skirt_04_S_l, bone2Name=dyn_Skirt_03_S_l, swing1LimitDegrees=2.5, swing2LimitDegrees=30, twistLimitDegrees=0}
 {bone1Name=dyn_Skirt_05_S_l, bone2Name=dyn_Skirt_04_S_l, swing1LimitDegrees=2.5, swing2LimitDegrees=35, twistLimitDegrees=0}
 {
  bone1Name=offset_Skirt_01_S_r
  bone2Name=spine_01
  swing1LimitDegrees=45
  swing1Motion=Free
  swing2LimitDegrees=45
  swing2Motion=Free
  twistLimitDegrees=45
  twistMotion=Free
 }
 {bone1Name=dyn_Skirt_01_S_r, bone2Name=offset_Skirt_01_S_r, swing1LimitDegrees=2.5, swing2LimitDegrees=15, twistLimitDegrees=0}
 {bone1Name=dyn_Skirt_02_S_r, bone2Name=dyn_Skirt_01_S_r, swing1LimitDegrees=2.5, swing2LimitDegrees=20, twistLimitDegrees=0}
 {bone1Name=dyn_Skirt_03_S_r, bone2Name=dyn_Skirt_02_S_r, swing1LimitDegrees=2.5, swing2LimitDegrees=25, twistLimitDegrees=0}
 {bone1Name=dyn_Skirt_04_S_r, bone2Name=dyn_Skirt_03_S_r, swing1LimitDegrees=2.5, swing2LimitDegrees=30, twistLimitDegrees=0}
 {bone1Name=dyn_Skirt_05_S_r, bone2Name=dyn_Skirt_04_S_r, swing1LimitDegrees=2.5, swing2LimitDegrees=35, twistLimitDegrees=0}
 {
  bone1Name=ball_r
  bone2Name=foot_r
  swing1LimitDegrees=45
  swing1Motion=Free
  swing2LimitDegrees=45
  swing2Motion=Free
  twistLimitDegrees=45
  twistMotion=Free
 }
 {
  bone1Name=ball_l
  bone2Name=foot_l
  swing1LimitDegrees=45
  swing1Motion=Free
  swing2LimitDegrees=45
  swing2Motion=Free
  twistLimitDegrees=45
  twistMotion=Free
 }
]
