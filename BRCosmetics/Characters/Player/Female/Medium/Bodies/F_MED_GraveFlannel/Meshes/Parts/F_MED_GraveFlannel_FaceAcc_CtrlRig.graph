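# Control Rig for the female GraveFlannel face accessory that applies pose-driven corrective offsets to hair bones.
# The first sequence branch derives look-up weights from head-to-neck swing/twist rotation and crouch weights from pelvis height, shaped by curves.
# The second layers weighted bone offsets, combining those weights with smoothed aiming, movement and pistol/pickaxe states.
# A rotation limit on the left mid strand runs midway through that stack.
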ControlRigBlueprint /BRCosmetics/Characters/Player/Female/Medium/Bodies/F_MED_GraveFlannel/Meshes/Parts/F_MED_GraveFlannel_FaceAcc_CtrlRig
graphs [
 Multiply_2_ContainedGraph {nodes [
   n1 = RigVMFunctionEntryNode()
   Multiply = &RigVMFunction_MathDoubleMul(A=@n1.A, B=@n1.B)
   RigVMFunctionReturnNode(Result=&RigVMFunction_MathDoubleMul(B=@n1.C, A=@Multiply.Result).Result)
  ]}
 Multiply_2_ContainedGraph {nodes [
   n1 = RigVMFunctionEntryNode()
   Multiply = &RigVMFunction_MathDoubleMul(A=@n1.A, B=@n1.B)
   RigVMFunctionReturnNode(Result=&RigVMFunction_MathDoubleMul(B=@n1.D, A=&RigVMFunction_MathDoubleMul(B=@n1.C, A=@Multiply.Result).Result).Result)
  ]}
 Multiply_3_ContainedGraph {nodes [
   n1 = RigVMFunctionEntryNode()
   RigVMFunctionReturnNode(Result=&RigVMFunction_MathDoubleMul(B=@n1.C, A=&RigVMFunction_MathDoubleMul(A=@n1.A, B=@n1.B).Result).Result)
  ]}
 Multiply_3_ContainedGraph {nodes [
   n1 = RigVMFunctionEntryNode()
   RigVMFunctionReturnNode(Result=&RigVMFunction_MathDoubleMul(B=@n1.C, A=&RigVMFunction_MathDoubleMul(A=@n1.A, B=@n1.B).Result).Result)
  ]}
 New Function_1_ContainedGraph {nodes [
   n1 = RigVMFunctionEntryNode()
   RigVMFunction_MathQuaternionToEuler = &RigVMFunction_MathQuaternionToEuler(RotationOrder=ZYX, Value=&RigUnit_GetTransform(Space=LocalSpace, bInitial=False, Item=@n1.Item).Transform.Rotation)
   Clamp = &RigVMFunction_MathIntClamp(Value=@n1.Rotate Index)
   n2 = RigVMDispatchNode(
    Values=(0,0,0)
    Index=@Clamp.Result
    Values.0=@RigVMFunction_MathQuaternionToEuler.Result.X
    Values.1=@RigVMFunction_MathQuaternionToEuler.Result.Y
    Values.2=@RigVMFunction_MathQuaternionToEuler.Result.Z
   ) {template="DISPATCH_RigVMDispatch_SelectInt32(in Index,in Values,out Result)"}
   n3 = RigVMVariableNode() {variable=Max Rotate}
   LessEqual = /Script/RigVM.RigVMFunction_MathDoubleLessEqual(A=@n2.Result, B=RigVMVariableNode() {variable=Max Rotate}.Value) {is=struct RigVMUnitNode, template="LessEqual::Execute(in A,in B,out Result)"}
   RigVMFunction_MathBoolToInteger = &RigVMFunction_MathBoolToInteger(Value=RigVMDispatchNode(
     Values=(False,False)
     Values.0=/Script/RigVM.RigVMFunction_MathDoubleGreaterEqual(A=@n2.Result, B=@n3.Value) {is=struct RigVMUnitNode, template="GreaterEqual::Execute(in A,in B,out Result)"}.Result
     Values.1=@LessEqual.Result
     Index=&RigVMFunction_MathBoolToInteger(Value=@n1.Is Lower limit).Result
    ) {template="DISPATCH_RigVMDispatch_SelectInt32(in Index,in Values,out Result)"}.Result)
   n4 = RigVMDispatchNode(Values=(0,0), Index=@RigVMFunction_MathBoolToInteger.Result, Values.0=@n2.Result, Values.1=RigVMVariableNode() {variable=Max Rotate}.Value) {template="DISPATCH_RigVMDispatch_SelectInt32(in Index,in Values,out Result)"}
   RigVMFunction_MathQuaternionFromEuler = &RigVMFunction_MathQuaternionFromEuler(Euler=RigVMDispatchNode(
     Values=((X=0,Y=0,Z=0),(X=0,Y=0,Z=0),(X=0,Y=0,Z=0))
     Values.1.X=@RigVMFunction_MathQuaternionToEuler.Result.X
     Values.2.X=@RigVMFunction_MathQuaternionToEuler.Result.X
     Values.0.Y=@RigVMFunction_MathQuaternionToEuler.Result.Y
     Values.2.Y=@RigVMFunction_MathQuaternionToEuler.Result.Y
     Values.0.Z=@RigVMFunction_MathQuaternionToEuler.Result.Z
     Values.1.Z=@RigVMFunction_MathQuaternionToEuler.Result.Z
     Values.0.X=@n4.Result
     Values.1.Y=@n4.Result
     Values.2.Z=@n4.Result
     Index=@Clamp.Result
    ) {template="DISPATCH_RigVMDispatch_SelectInt32(in Index,in Values,out Result)"}.Result)
   > Set Transform = &RigUnit_SetRotation(Weight=1, bInitial=False, Item=@n1.Item, Space=@n1.Space, Value=@RigVMFunction_MathQuaternionFromEuler.Result)
   > RigVMFunctionReturnNode()
  ]}
 New Function_ContainedGraph {nodes [
   n1 = RigVMFunctionEntryNode()
   RigVMFunction_MathBoolToInteger = &RigVMFunction_MathBoolToInteger(Value=@n1.onlyTwist)
   RigVMFunction_MathQuaternionSwingTwist = &RigVMFunction_MathQuaternionSwingTwist(
    TwistAxis=@n1.twistAxis
    Input=&RigUnit_GetRelativeTransformForItem(bChildInitial=False, bParentInitial=False, Child=@n1.child, Parent=@n1.parent).RelativeTransform.Rotation
   )
   n2 = RigVMDispatchNode(
    Values=((),())
    Index=@RigVMFunction_MathBoolToInteger.Result
    Values.0=@RigVMFunction_MathQuaternionSwingTwist.Swing
    Values.1=@RigVMFunction_MathQuaternionSwingTwist.Twist
   ) {template="DISPATCH_RigVMDispatch_SelectInt32(in Index,in Values,out Result)"}
   RigVMFunction_MathQuaternionToEuler = &RigVMFunction_MathQuaternionToEuler(Value=@n2.Result, RotationOrder=@n1.Rotation Order)
   > DISPATCH_RigVMDispatch_Print = RigVMDispatchNode(ScreenColor=(R=1,G=1,B=1,A=1), ScreenDuration=1.5, Enabled=@n1.debugRotation, Value=@RigVMFunction_MathQuaternionToEuler.Result) {template="DISPATCH_RigVMDispatch_Print(in Prefix,in Value,in Enabled,in ScreenDuration,in ScreenColor)"}
   Cast = RigVMDispatchNode(Value=@n1.triggerAxis) {template="DISPATCH_RigVMDispatch_CastEnumToInt(in Value,out Result)"}
   DISPATCH_RigVMDispatch_SelectInt32 = RigVMDispatchNode(
    Values=(0,0,0)
    Index=@Cast.Result
    Values.0=@RigVMFunction_MathQuaternionToEuler.Result.X
    Values.1=@RigVMFunction_MathQuaternionToEuler.Result.Y
    Values.2=@RigVMFunction_MathQuaternionToEuler.Result.Z
   ) {template="DISPATCH_RigVMDispatch_SelectInt32(in Index,in Values,out Result)"}
   Remap = &RigVMFunction_MathDoubleRemap(
    bClamp=False
    Value=@DISPATCH_RigVMDispatch_SelectInt32.Result
    TargetMinimum=@n1.outputMinMax.X
    TargetMaximum=@n1.outputMinMax.Y
    SourceMinimum=@n1.rotationMinMax.X
    SourceMaximum=@n1.rotationMinMax.Y
   )
   Absolute = &RigVMFunction_MathDoubleAbs(Value=@Remap.Result)
   Multiply = &RigVMFunction_MathDoubleMul(
    A=&RigVMFunction_MathDoubleClamp(Value=RigVMDispatchNode(True=@Absolute.Result, Condition=@n1.absoluteValue, False=@Remap.Result) {template="DISPATCH_RigVMDispatch_If(in Condition,in True,in False,out Result)"}.Result).Result
    B=@n1.premult
   )
   > RigVMFunctionReturnNode(outputQuat=@n2.Result, outputWeight=@Multiply.Result)
  ]}
 New Function_ContainedGraph {nodes [
   n1 = RigVMFunctionEntryNode()
   RigVMFunction_MathBoolToInteger = &RigVMFunction_MathBoolToInteger(Value=@n1.onlyTwist)
   RigVMFunction_MathQuaternionSwingTwist = &RigVMFunction_MathQuaternionSwingTwist(
    TwistAxis=@n1.twistAxis
    Input=&RigUnit_GetRelativeTransformForItem(bChildInitial=False, bParentInitial=False, Child=@n1.child, Parent=@n1.parent).RelativeTransform.Rotation
   )
   n2 = RigVMDispatchNode(
    Values=((),())
    Index=@RigVMFunction_MathBoolToInteger.Result
    Values.0=@RigVMFunction_MathQuaternionSwingTwist.Swing
    Values.1=@RigVMFunction_MathQuaternionSwingTwist.Twist
   ) {template="DISPATCH_RigVMDispatch_SelectInt32(in Index,in Values,out Result)"}
   RigVMFunction_MathQuaternionToEuler = &RigVMFunction_MathQuaternionToEuler(Value=@n2.Result, RotationOrder=@n1.Rotation Order)
   > DISPATCH_RigVMDispatch_Print = RigVMDispatchNode(ScreenColor=(R=1,G=1,B=1,A=1), ScreenDuration=1.5, Enabled=@n1.debugRotation, Value=@RigVMFunction_MathQuaternionToEuler.Result) {template="DISPATCH_RigVMDispatch_Print(in Prefix,in Value,in Enabled,in ScreenDuration,in ScreenColor)"}
   Cast = RigVMDispatchNode(Value=@n1.triggerAxis) {template="DISPATCH_RigVMDispatch_CastEnumToInt(in Value,out Result)"}
   DISPATCH_RigVMDispatch_SelectInt32 = RigVMDispatchNode(
    Values=(0,0,0)
    Index=@Cast.Result
    Values.0=@RigVMFunction_MathQuaternionToEuler.Result.X
    Values.1=@RigVMFunction_MathQuaternionToEuler.Result.Y
    Values.2=@RigVMFunction_MathQuaternionToEuler.Result.Z
   ) {template="DISPATCH_RigVMDispatch_SelectInt32(in Index,in Values,out Result)"}
   Remap = &RigVMFunction_MathDoubleRemap(
    bClamp=False
    Value=@DISPATCH_RigVMDispatch_SelectInt32.Result
    TargetMinimum=@n1.outputMinMax.X
    TargetMaximum=@n1.outputMinMax.Y
    SourceMinimum=@n1.rotationMinMax.X
    SourceMaximum=@n1.rotationMinMax.Y
   )
   Absolute = &RigVMFunction_MathDoubleAbs(Value=@Remap.Result)
   Multiply = &RigVMFunction_MathDoubleMul(
    A=&RigVMFunction_MathDoubleClamp(Value=RigVMDispatchNode(True=@Absolute.Result, Condition=@n1.absoluteValue, False=@Remap.Result) {template="DISPATCH_RigVMDispatch_If(in Condition,in True,in False,out Result)"}.Result).Result
    B=@n1.premult
   )
   > RigVMFunctionReturnNode(outputQuat=@n2.Result, outputWeight=@Multiply.Result)
  ]}
 New Function_ContainedGraph {nodes [
   n1 = RigVMFunctionEntryNode()
   RigVMFunction_MathBoolToInteger = &RigVMFunction_MathBoolToInteger(Value=@n1.onlyTwist)
   RigVMFunction_MathQuaternionSwingTwist = &RigVMFunction_MathQuaternionSwingTwist(
    TwistAxis=@n1.twistAxis
    Input=&RigUnit_GetRelativeTransformForItem(bChildInitial=False, bParentInitial=False, Child=@n1.child, Parent=@n1.parent).RelativeTransform.Rotation
   )
   n2 = RigVMDispatchNode(
    Values=((),())
    Index=@RigVMFunction_MathBoolToInteger.Result
    Values.0=@RigVMFunction_MathQuaternionSwingTwist.Swing
    Values.1=@RigVMFunction_MathQuaternionSwingTwist.Twist
   ) {template="DISPATCH_RigVMDispatch_SelectInt32(in Index,in Values,out Result)"}
   RigVMFunction_MathQuaternionToEuler = &RigVMFunction_MathQuaternionToEuler(Value=@n2.Result, RotationOrder=@n1.Rotation Order)
   > DISPATCH_RigVMDispatch_Print = RigVMDispatchNode(ScreenColor=(R=1,G=1,B=1,A=1), ScreenDuration=1.5, Enabled=@n1.debugRotation, Value=@RigVMFunction_MathQuaternionToEuler.Result) {template="DISPATCH_RigVMDispatch_Print(in Prefix,in Value,in Enabled,in ScreenDuration,in ScreenColor)"}
   Cast = RigVMDispatchNode(Value=@n1.triggerAxis) {template="DISPATCH_RigVMDispatch_CastEnumToInt(in Value,out Result)"}
   DISPATCH_RigVMDispatch_SelectInt32 = RigVMDispatchNode(
    Values=(0,0,0)
    Index=@Cast.Result
    Values.0=@RigVMFunction_MathQuaternionToEuler.Result.X
    Values.1=@RigVMFunction_MathQuaternionToEuler.Result.Y
    Values.2=@RigVMFunction_MathQuaternionToEuler.Result.Z
   ) {template="DISPATCH_RigVMDispatch_SelectInt32(in Index,in Values,out Result)"}
   Remap = &RigVMFunction_MathDoubleRemap(
    bClamp=False
    Value=@DISPATCH_RigVMDispatch_SelectInt32.Result
    TargetMinimum=@n1.outputMinMax.X
    TargetMaximum=@n1.outputMinMax.Y
    SourceMinimum=@n1.rotationMinMax.X
    SourceMaximum=@n1.rotationMinMax.Y
   )
   Absolute = &RigVMFunction_MathDoubleAbs(Value=@Remap.Result)
   Multiply = &RigVMFunction_MathDoubleMul(
    A=&RigVMFunction_MathDoubleClamp(Value=RigVMDispatchNode(True=@Absolute.Result, Condition=@n1.absoluteValue, False=@Remap.Result) {template="DISPATCH_RigVMDispatch_If(in Condition,in True,in False,out Result)"}.Result).Result
    B=@n1.premult
   )
   > RigVMFunctionReturnNode(outputQuat=@n2.Result, outputWeight=@Multiply.Result)
  ]}
 RigVMFunctionLibrary {links [], nodes []}
 RigVMModel {nodes [
   RigUnit_BeginExecution = &RigUnit_BeginExecution()
   RigVMFunction_Sequence = &RigVMFunction_Sequence(ExecuteContext=@RigUnit_BeginExecution.ExecutePin)
   DeltaRotation v4_1 = RigVMCollapseNode(
    Rotation Order=ZYX
    child=(Type=Bone,Name="head")
    debugRotation=false
    outputMinMax=(X=0,Y=1)
    parent=(Type=Bone,Name="neck_01")
    premult=1
    rotationMinMax=(X=-20,Y=-43)
    triggerAxis=Z
    twistAxis=(X=1,Y=0,Z=0)
    ExecuteContext=@RigVMFunction_Sequence.A
   )
   n1 = &RigVMFunction_AnimEvalRichCurve(
    Curve="(EditorCurveData=(Keys=((InterpMode=RCIM_Cubic),(InterpMode=RCIM_Cubic,TangentMode=RCTM_User,Time=0.5,Value=0.25,ArriveTangent=0.9828,LeaveTangent=0.9828),(InterpMode=RCIM_Cubic,TangentMode=RCTM_User,Time=1,Value=1,ArriveTangent=1.918,LeaveTangent=1.918))) ...(truncated)"
    SourceMinimum=0
    TargetMinimum=0
    Value=@"DeltaRotation v4_1".outputWeight
   )
   > RigVMVariableNode(Value=@n1.Result) {variable=HeadLookingUp}
   > DeltaRotation v4_2 = RigVMCollapseNode(
    Rotation Order=ZYX
    child=(Type=Bone,Name="head")
    debugRotation=false
    outputMinMax=(X=0,Y=1)
    parent=(Type=Bone,Name="neck_01")
    premult=1
    rotationMinMax=(X=-50,Y=-80)
    triggerAxis=Z
    twistAxis=(X=1,Y=0,Z=0)
   )
   > RigVMVariableNode(Value=@"DeltaRotation v4_2".outputWeight) {variable=HeadLookingUp_Extreme}
   > DeltaRotation v4_3 = RigVMCollapseNode(
    Rotation Order=ZYX
    child=(Type=Bone,Name="head")
    debugRotation=false
    outputMinMax=(X=0,Y=1)
    parent=(Type=Bone,Name="neck_01")
    premult=1
    rotationMinMax=(X=-20,Y=-53)
    triggerAxis=Z
    twistAxis=(X=1,Y=0,Z=0)
   )
   > RigVMVariableNode(Value=@"DeltaRotation v4_3".outputWeight) {variable=headLookingUp_Crouching}
   n2 = &RigVMFunction_AnimEvalRichCurve(
    Curve=(EditorCurveData=(Keys=((InterpMode=RCIM_Cubic),(InterpMode=RCIM_Cubic,Time=0.5,Value=1),(Time=0.8))))
    SourceMaximum=2
    SourceMinimum=0
    TargetMinimum=0
    Value=&RigVMFunction_MathDoubleRemap(
     SourceMaximum=75
     SourceMinimum=45
     TargetMaximum=1
     TargetMinimum=0
     bClamp=false
     Value=&RigUnit_GetRelativeTransformForItem(Child=(Type=Bone,Name="pelvis"), Parent=(Type=Bone,Name="root"), bChildInitial=False, bParentInitial=False).RelativeTransform.Translation.Z
    ).Result
   )
   > RigVMVariableNode(Value=@n2.Result) {variable=isCrouchWalking}
   n3 = &RigUnit_GetRelativeTransformForItem(Child=(Type=Bone,Name="pelvis"), Parent=(Type=Bone,Name="root"), bChildInitial=False, bParentInitial=False)
   > RigVMVariableNode(Value=&RigVMFunction_AnimEvalRichCurve(
     Curve=(EditorCurveData=(Keys=((),(Time=1,Value=1))))
     SourceMinimum=0
     TargetMinimum=0
     Value=&RigVMFunction_MathDoubleRemap(SourceMaximum=85, SourceMinimum=44, TargetMaximum=0, TargetMinimum=1, bClamp=false, Value=@n3.RelativeTransform.Translation.Z).Result
    ).Result) {variable=isCrouching}
   n4 = &RigVMFunction_AnimEvalRichCurve(
    Curve=(EditorCurveData=(Keys=((InterpMode=RCIM_Cubic),(InterpMode=RCIM_Cubic,Time=0.5,Value=1),(Time=0.8))))
    SourceMaximum=2
    SourceMinimum=0
    TargetMinimum=0
    Value=&RigVMFunction_MathDoubleRemap(
     SourceMaximum=60
     SourceMinimum=45
     TargetMaximum=0
     TargetMinimum=1
     bClamp=false
     Value=&RigUnit_GetRelativeTransformForItem(Child=(Type=Bone,Name="pelvis"), Parent=(Type=Bone,Name="root"), bChildInitial=False, bParentInitial=False).RelativeTransform.Translation.Z
    ).Result
   )
   > RigVMVariableNode(Value=@n4.Result) {variable=isJustCrouching}
   &RigUnit_ModifyTransforms(
    ItemToModify="((Item=(Type=Bone,Name=\"base_hairBackMain_01\"),Transform=(Rotation=(X=0,Y=0,Z=-0.0872,W=0.9962),Translation=(X=-2,Y=-2,Z=0),Scale3D=(X=1,Y=1,Z=1))),(Item=(Type=Bone,Name=\"base_frontStrand_01_l\"),Transform=(Rotation=(X=0,Y=0,Z=-0.1305,W=0.9914),Translation= ...(truncated)"
    WeightMinimum=0
    Weight=RigVMVariableNode() {variable=HeadLookingUp}.Value
    ExecutePin=@RigVMFunction_Sequence.B
   )
   > &RigUnit_ModifyTransforms(
    ItemToModify=((Item=(Type=Bone,Name="offset_hairPushJnt"),Transform=(Rotation=(X=-0.0226,Y=-0.0842,Z=-0.2578,W=0.9623),Translation=(X=0,Y=-1,Z=0),Scale3D=(X=1,Y=1,Z=1))))
    WeightMinimum=0
    Weight=RigVMVariableNode() {variable=HeadLookingUp}.Value
   )
   n5 = &RigVMFunction_AlphaInterp(
    InterpSpeedDecreasing=10
    InterpSpeedIncreasing=10
    Scale=1
    bClampResult=False
    Value=&RigVMFunction_MathBoolToFloat(Value=RigVMVariableNode() {variable=isAiming}.Value).Result
   )
   RigVMFunction_AnimEvalRichCurve = &RigVMFunction_AnimEvalRichCurve(
    Curve="(EditorCurveData=(Keys=((InterpMode=RCIM_Cubic),(InterpMode=RCIM_Cubic,TangentMode=RCTM_User,Time=0.5001,Value=0.25,ArriveTangent=0.8193,LeaveTangent=0.8193),(InterpMode=RCIM_Cubic,TangentMode=RCTM_User,Time=1,Value=1,ArriveTangent=1.7002,LeaveTangent=1.70 ...(truncated)"
    SourceMinimum=0
    TargetMinimum=0
    Value=RigVMVariableNode() {variable=HeadLookingUp_Extreme}.Value
   )
   n6 = &RigVMFunction_MathDoubleMul(A=@n5.Result, B=@RigVMFunction_AnimEvalRichCurve.Result)
   > &RigUnit_ModifyTransforms(
    ItemToModify="((Item=(Type=Bone,Name=\"base_hairBackMain_01\"),Transform=(Rotation=(X=0,Y=0,Z=0.0436,W=0.999),Translation=(X=-1,Y=-1.5,Z=3),Scale3D=(X=0.9,Y=1,Z=0.8))),(Item=(Type=Bone,Name=\"base_frontStrand_01_l\"),Transform=(Rotation=(X=0,Y=0,Z=0,W=1),Translation=(X=0,Y= ...(truncated)"
    WeightMinimum=0
    Weight=@n6.Result
   )
   > &RigUnit_ModifyTransforms(
    ItemToModify=((Item=(Type=Bone,Name="offset_hairPushJnt"),Transform=(Rotation=(X=0,Y=-0.0529,Z=0,W=0.9986),Translation=(X=-5,Y=5,Z=0.041),Scale3D=(X=1,Y=1,Z=1))))
    WeightMinimum=0
    Weight=@n6.Result
   )
   RigVMFunction_MathBoolNot = &RigVMFunction_MathBoolNot(Value=RigVMVariableNode() {variable=isAiming}.Value)
   n7 = &RigVMFunction_AlphaInterp(
    InterpSpeedDecreasing=10
    InterpSpeedIncreasing=10
    Scale=1
    bClampResult=False
    Value=&RigVMFunction_MathBoolToFloat(Value=@RigVMFunction_MathBoolNot.Result).Result
   )
   n8 = &RigVMFunction_AnimEvalRichCurve(
    Curve="(EditorCurveData=(Keys=((InterpMode=RCIM_Cubic,TangentMode=RCTM_User,ArriveTangent=0.5739,LeaveTangent=0.5739),(InterpMode=RCIM_Cubic,Time=0.5001,Value=0.4,ArriveTangent=1,LeaveTangent=1),(InterpMode=RCIM_Cubic,TangentMode=RCTM_User,Time=1,Value=1,ArriveTa ...(truncated)"
    SourceMinimum=0
    TargetMinimum=0
    Value=RigVMVariableNode() {variable=HeadLookingUp_Extreme}.Value
   )
   n9 = &RigVMFunction_MathDoubleMul(A=@n7.Result, B=@n8.Result)
   > &RigUnit_ModifyTransforms(
    ItemToModify="((Item=(Type=Bone,Name=\"base_hairBackMain_01\"),Transform=(Rotation=(X=0,Y=0,Z=0,W=1),Translation=(X=-1,Y=-1.5,Z=3),Scale3D=(X=0.9,Y=1,Z=0.8))),(Item=(Type=Bone,Name=\"base_frontStrand_01_l\"),Transform=(Rotation=(X=0,Y=0,Z=-0.0872,W=0.9962),Translation=(X=0, ...(truncated)"
    WeightMinimum=0
    Weight=@n9.Result
   )
   > &RigUnit_ModifyTransforms(
    ItemToModify=((Item=(Type=Bone,Name="offset_hairPushJnt"),Transform=(Rotation=(X=0,Y=-0.0529,Z=0,W=0.9986),Translation=(X=-5,Y=5,Z=0.041),Scale3D=(X=1,Y=1,Z=1))))
    WeightMinimum=0
    Weight=@n9.Result
   )
   > LimitValue = RigVMCollapseNode(Is Lower limit=true, Item=(Type=Bone,Name="base_MidStrand_01_l"), Max Rotate=-90, Rotate Index=2, Space=LocalSpace)
   AlphaInterp = &RigVMFunction_AlphaInterp(InterpSpeedDecreasing=8, InterpSpeedIncreasing=2, Scale=1, bClampResult=False, Value=RigVMVariableNode() {variable=isMoving}.Value)
   Multiply = RigVMAggregateNode(
    C=@AlphaInterp.Result
    A=&RigVMFunction_MathDoubleRemap(SourceMaximum=0.6, SourceMinimum=0, TargetMaximum=1, TargetMinimum=0, Value=RigVMVariableNode() {variable=HeadLookingUp_Extreme}.Value).Result
    B=RigVMVariableNode() {variable=isCrouchWalking}.Value
   )
   > &RigUnit_ModifyTransforms(
    ItemToModify="((Item=(Type=Bone,Name=\"base_hairBackMain_00\"),Transform=(Rotation=(X=0,Y=0,Z=-0.0872,W=0.9962),Translation=(X=0,Y=0,Z=0),Scale3D=(X=1,Y=1,Z=1))),(Item=(Type=Bone,Name=\"base_frontStrand_01_r\"),Transform=(Rotation=(X=0,Y=0,Z=-0.0436,W=0.999),Translation=(X= ...(truncated)"
    WeightMinimum=0
    Weight=@Multiply.Result
   )
   n10 = &RigVMFunction_AlphaInterp(InterpSpeedDecreasing=8, InterpSpeedIncreasing=2, Scale=1, bClampResult=False, Value=RigVMVariableNode() {variable=isMoving}.Value)
   n11 = &RigVMFunction_AlphaInterp(
    InterpSpeedDecreasing=8
    InterpSpeedIncreasing=8
    Scale=1
    bClampResult=False
    Value=&RigVMFunction_MathBoolToFloat(Value=RigVMVariableNode() {variable=isAiming}.Value).Result
   )
   Multiply_6 = RigVMAggregateNode(
    C=@n10.Result
    D=@n11.Result
    A=&RigVMFunction_MathDoubleRemap(SourceMaximum=0.6, SourceMinimum=0, TargetMaximum=1, TargetMinimum=0, Value=RigVMVariableNode() {variable=HeadLookingUp_Extreme}.Value).Result
    B=RigVMVariableNode() {variable=isCrouchWalking}.Value
   )
   > &RigUnit_ModifyTransforms(
    ItemToModify="((Item=(Type=Bone,Name=\"base_hairBackMain_00\"),Transform=(Rotation=(X=0,Y=0,Z=-0.3007,W=0.9537),Translation=(X=0,Y=0,Z=0),Scale3D=(X=1,Y=1,Z=1))),(Item=(Type=Bone,Name=\"base_frontStrand_01_r\"),Transform=(Rotation=(X=0,Y=0,Z=-0.1736,W=0.9848),Translation=(X ...(truncated)"
    WeightMinimum=0
    Weight=@Multiply_6.Result
   )
   n12 = &RigVMFunction_AlphaInterp(
    InterpSpeedDecreasing=4
    InterpSpeedIncreasing=4
    Scale=1
    bClampResult=False
    Value=&RigVMFunction_MathBoolToFloat(Value=RigVMVariableNode() {variable=PistolOut}.Value).Result
   )
   n13 = &RigVMFunction_MathDoubleMul(
    A=@Multiply_6.Result
    B=&RigVMFunction_MathDoubleRemap(SourceMaximum=1, SourceMinimum=0, TargetMaximum=0.6, TargetMinimum=1, bClamp=False, Value=@n12.Result).Result
   )
   > &RigUnit_ModifyTransforms(
    ItemToModify="((Item=(Type=Bone,Name=\"base_MidStrand_01_l\"),Transform=(Rotation=(X=0,Y=0,Z=-0.5,W=0.866),Translation=(X=0,Y=0,Z=0),Scale3D=(X=1,Y=1,Z=1))),(Item=(Type=Bone,Name=\"dyn_MidStrand_02_l\"),Transform=(Rotation=(X=0,Y=0,Z=-0.2588,W=0.9659),Translation=(X=0,Y=0,Z ...(truncated)"
    WeightMinimum=0
    Weight=@n13.Result
   )
   > &RigUnit_ModifyTransforms(
    ItemToModify="((Item=(Type=Bone,Name=\"dyn_hairBackMain_02\"),Transform=(Rotation=(X=0,Y=0,Z=0.3827,W=0.9239),Translation=(X=0,Y=0,Z=0),Scale3D=(X=1,Y=1,Z=1))),(Item=(Type=Bone,Name=\"dyn_midStrandEnd_C_l\"),Transform=(Rotation=(X=0,Y=0,Z=0.342,W=0.9397),Translation=(X=0,Y= ...(truncated)"
    WeightMinimum=0
    Weight=@n13.Result
   )
   RigVMFunction_MathBoolToFloat = &RigVMFunction_MathBoolToFloat(Value=RigVMVariableNode() {variable=isAiming}.Value)
   n14 = &RigVMFunction_AlphaInterp(InterpSpeedDecreasing=8, InterpSpeedIncreasing=8, Scale=1, bClampResult=False, Value=@RigVMFunction_MathBoolToFloat.Result)
   Multiply_2 = RigVMAggregateNode(
    B=@n14.Result
    A=&RigVMFunction_MathDoubleRemap(SourceMaximum=1, SourceMinimum=0, TargetMaximum=1, TargetMinimum=0.3, bClamp=False, Value=RigVMVariableNode() {variable=isCrouching}.Value).Result
    C=RigVMVariableNode() {variable=headLookingUp_Crouching}.Value
   )
   > &RigUnit_ModifyTransforms(
    ItemToModify="((Item=(Type=Bone,Name=\"base_hairBackMain_00\"),Transform=(Rotation=(X=-0.2566,Y=0.0338,Z=-0.1261,W=0.9577),Translation=(X=0,Y=0,Z=0),Scale3D=(X=1,Y=1,Z=1))),(Item=(Type=Bone,Name=\"base_frontStrand_01_r\"),Transform=(Rotation=(X=-0.0298,Y=-0.0819,Z=-0.3407,W ...(truncated)"
    WeightMinimum=0
    Weight=@Multiply_2.Result
   )
   > &RigUnit_ModifyTransforms(
    ItemToModify="((Item=(Type=Bone,Name=\"dyn_frontStrandEnd_A_r\"),Transform=(Rotation=(X=0,Y=0,Z=0.5,W=0.866),Translation=(X=-1,Y=-0.25,Z=0),Scale3D=(X=1,Y=1,Z=1))),(Item=(Type=Bone,Name=\"dyn_frontStrandEnd_B_r\"),Transform=(Rotation=(X=0,Y=0,Z=0.5,W=0.866),Translation=(X=0 ...(truncated)"
    WeightMinimum=0
    Weight=@Multiply_2.Result
   )
   RigVMFunction_MathBoolOr = &RigVMFunction_MathBoolOr(A=RigVMVariableNode() {variable=PickaxeOut}.Value, B=RigVMVariableNode() {variable=PistolOut}.Value)
   n15 = &RigVMFunction_AlphaInterp(
    InterpSpeedDecreasing=8
    InterpSpeedIncreasing=8
    Scale=1
    bClampResult=False
    Value=&RigVMFunction_MathBoolToFloat(Value=&RigVMFunction_MathBoolNot(Value=@RigVMFunction_MathBoolOr.Result).Result).Result
   )
   Multiply_1 = RigVMAggregateNode(C=@n15.Result, A=@Multiply_2.Result, B=RigVMVariableNode() {variable=isJustCrouching}.Value)
   > &RigUnit_ModifyTransforms(
    ItemToModify="((Item=(Type=Bone,Name=\"dyn_frontStrandEnd_A_r\"),Transform=(Rotation=(X=0,Y=0,Z=-0.1736,W=0.9848),Translation=(X=0,Y=0,Z=0),Scale3D=(X=1,Y=1,Z=1))),(Item=(Type=Bone,Name=\"dyn_frontStrandEnd_B_r\"),Transform=(Rotation=(X=0,Y=0,Z=0,W=1),Translation=(X=0,Y=0,Z ...(truncated)"
    WeightMinimum=0
    Weight=@Multiply_1.Result
   )
   &RigUnit_PrepareForExecution()
   addFKCtrl = RigVMFunctionReferenceNode(
    Custom_FKCtrl=()
    ExecuteContext=()
    Options_FKCtrl="(pelvis_1_#1=False,spine_3_#2=true,clavicle_5_#3=False,arm_7_#4=False,leg_9_#5=False,head_11_46DC15EF465CD ...(truncated)"
    Shape_Settings="(pelvis_2_#6=(bVisible=True,Name=\"Hexagon_Thin\",Color=(R=0.0117,G=0.8281,B=1,A=1),Transform=(Rotation=(X=0,Y=-0.7071,Z=0,W=0.7071),Translation=(X=2,Y=0,Z=0),Scale3D=(X=5,Y=5,Z=7))),spine_4_#7=(bVi ...(truncated)"
   )
   setFKCtrl = RigVMFunctionReferenceNode(ExecuteContext=(), setFK_Ctrl=false)
  ]}
]
hierarchy [
 root {loc=<0 0 0>, rot=<0 0 0 1>}
 pelvis {parent=root, loc=<0 -0.5301 98.0571>, rot=<0.0127 0.707 -0.0127 -0.707>}
 spine_01 {parent=pelvis, loc=<6.1208 -0.9328 0>, rot=<0 0 0.034 -0.9994>}
 spine_02 {parent=spine_01, loc=<6.5997 -0.2095 0>, rot=<0 0 0.0238 -0.9997>}
 spine_03 {parent=spine_02, loc=<6.4914 -0.1237 0>, rot=<0 0 0.0149 -0.9999>}
 spine_04 {parent=spine_03, loc=<6.9598 0.0794 0>, rot=<0 0 -0.0132 -0.9999>}
 spine_05 {parent=spine_04, loc=<6.9987 0 0>, rot=<0 0 -0.0015 -1>}
 neck_01 {parent=spine_05, loc=<11.3138 0.1529 0>, rot=<0 0 -0.2003 -0.9797>}
 neck_02 {parent=neck_01, loc=<4.5627 0 0>, rot=<0 0 0.0504 -0.9987>}
 head {parent=neck_02, loc=<5.3914 0 0>, rot=<0 0 0.115 -0.9934>}
 base_frontStrand_01_l {parent=head, loc=<7.149 4.2836 -7.0646>, rot=<0.0041 -1 0 0>}
 dyn_frontStrand_01_l {parent=base_frontStrand_01_l, loc=<0 0 0>, rot=<0 0 0 1>}
 dyn_frontStrand_02_l {parent=dyn_frontStrand_01_l, loc=<5.4533 -0.5731 1.2204>, rot=<0 0 0 1>}
 dyn_frontStrandEnd_A_l {parent=dyn_frontStrand_02_l, loc=<4.9256 1.7059 -0.7202>, rot=<0 0 0 1>}
 dyn_frontStrandEnd_B_l {parent=dyn_frontStrand_02_l, loc=<5.3812 -2.1388 -0.4613>, rot=<0 0 0 1>}
 dyn_frontStrandEnd_C_l {parent=dyn_frontStrand_02_l, loc=<5.6275 -2.8535 0.3509>, rot=<0 0 0 1>}
 base_MidStrand_01_l {parent=head, loc=<6.7394 -1.5896 -7.6922>, rot=<0.0041 -1 0 0>}
 dyn_MidStrand_01_l {parent=base_MidStrand_01_l, loc=<0 0 0>, rot=<0 0 0 1>}
 dyn_MidStrand_02_l {parent=dyn_MidStrand_01_l, loc=<5.4409 -0.7171 1.1993>, rot=<0 0 0 1>}
 dyn_midStrandEnd_A_l {parent=dyn_MidStrand_02_l, loc=<5.358 0.4174 0.8522>, rot=<0 0 0 1>}
 dyn_midStrandEnd_B_l {parent=dyn_MidStrand_02_l, loc=<5.6536 -1.4829 0.0589>, rot=<0 0 0 1>}
 dyn_midStrandEnd_C_l {parent=dyn_MidStrand_02_l, loc=<4.6447 -3.3593 -1.0824>, rot=<0 0 0 1>}
 base_frontStrand_01_r {parent=head, loc=<3.9163 -3.8396 6.891>, rot=<0.0041 -1 0 0>}
 dyn_frontStrand_01_r {parent=base_frontStrand_01_r, loc=<0 0 0>, rot=<0 0 0 1>}
 dyn_frontStrand_02_r {parent=dyn_frontStrand_01_r, loc=<4.5378 -0.6909 0.416>, rot=<0 0 0 1>}
 dyn_frontStrandEnd_A_r {parent=dyn_frontStrand_02_r, loc=<3.098 1.5846 -1.2433>, rot=<0 0 0 1>}
 dyn_frontStrandEnd_B_r {parent=dyn_frontStrand_02_r, loc=<3.2576 -0.1066 -0.1329>, rot=<0 0 0 1>}
 dyn_frontStrandEnd_C_r {parent=dyn_frontStrand_02_r, loc=<3.2174 -2.0276 -0.4028>, rot=<0 0 0 1>}
 base_earring_ring {parent=head, loc=<0.2521 0.5889 -6.0326>, rot=<0.0039 -0.9415 0.3369 -0.0014>}
 dyn_earring_ring {parent=base_earring_ring, loc=<0 0 0>, rot=<0 0 0 1>}
 dyn_earring_dimond {parent=dyn_earring_ring, loc=<3.6459 0.0177 0.0111>, rot=<0.3609 0 0 -0.9326>}
 base_frontSingleStrand_01_l {parent=head, loc=<5.1771 6.478 -6.7429>, rot=<0.0041 -1 0 0>}
 dyn_frontSingleStrand_01_l {parent=base_frontSingleStrand_01_l, loc=<0 0 0>, rot=<0 0 0 1>}
 dyn_frontSingleStrand_02_l {parent=dyn_frontSingleStrand_01_l, loc=<2.5901 -0.2359 -0.0802>, rot=<0 0 0 1>}
 base_hairBackMain_00 {parent=head, loc=<5.2304 -6.5158 -0.2453>, rot=<0.0601 -0.9982 0.0007 -0.0072>}
 dyn_hairBackMain_00 {parent=base_hairBackMain_00, loc=<0 0 0>, rot=<0 0 0 1>}
 base_hairBackMain_01 {parent=dyn_hairBackMain_00, loc=<4.4931 0 0>, rot=<0 0 0 1>}
 dyn_hairBackMain_01 {parent=base_hairBackMain_01, loc=<0 0 0>, rot=<0 0 0 1>}
 dyn_hairBackMain_02 {parent=dyn_hairBackMain_01, loc=<3.54 0 0>, rot=<0.0023 -0.0178 0.0171 -0.9997>}
 base_hairPushJnt {parent=head, loc=<6.9705 -4.8226 -0.2212>, rot=<0.0601 -0.9982 0.0007 -0.0072>}
 offset_hairPushJnt {parent=base_hairPushJnt, loc=<0 0 0>, rot=<0 0 0 1>}
 head_neck01Delta_null {parent=neck_01, type=Null, loc=<9.9266 -0.5431 0>, rot=<0 0 -0.1649 0.9863>}
]
variables [
 HeadLookingUp {type=double, default=0}
 HeadLookingUp_Extreme {type=double, default=0}
 PickaxeOut {type=bool, default=False}
 PistolOut {type=bool, default=False}
 headLookingUp_Crouching {type=double, default=0}
 isAiming {type=bool, default=False}
 isCrouchWalking {type=float, default=0}
 isCrouching {type=float, default=0}
 isJustCrouching {type=float, default=0}
 isMoving {type=double, default=0}
]
